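# Animation Blueprint for the ClawPad HostFull male body, adding secondary motion to its sleeves, robe skirt, front panel and shawl.
# From the slope-posed pose, sleeves use either RigidBody or AnimDynamics, chosen by UseAnimDynamicSleeves, while RigidBody simulates the skirt.
# A per-bone blend merges these into a post-sim Control Rig driven by movement and state.
# AnimDynamics chains then layer panel, sleeve and shawl motion onto that cached result.
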AnimBlueprint /BRCosmetics/Characters/Player/Male/Medium/Bodies/M_MED_ClawPad_HostFull/Meshes/M_MED_ClawPad_HostFull_AnimBP
target_skeleton=&/M_MED_ClawPad_HostFull_Skeleton
graphs [
 AnimGraph {nodes [
   Output Pose(Result=Use cached pose 'Post Main Anim Dynamics'()) {layerGroup=None}
   n1 = "Transform (Modify) Bone - Bone: dyn_lowerarm_sleeve_r_MSTR"(
    ComponentPose=Local To Component(LocalPose=Use cached pose 'Post Slope Posing'())
    Alpha=1
    Translation="0,0,0"
    Rotation="0,0,30"
    Scale="1,1,1"
   ) {
    rotationMode=BMM_Additive
    rotationSpace=BCS_BoneSpace
    translationSpace=BCS_BoneSpace
    alphaBoolBlend {blendInTime=0.1, blendOption=Linear, blendOutTime=0.1}
    boneToModify {boneName=dyn_lowerarm_sleeve_r_MSTR}
    rotation {roll=30}
   }
   n2 = "Transform (Modify) Bone - Bone: dyn_lowerarm_sleeve_l_MSTR"(ComponentPose=@n1, Alpha=1, Translation="0,0,0", Rotation="0,0,30", Scale="1,1,1") {
    rotationMode=BMM_Additive
    rotationSpace=BCS_BoneSpace
    translationSpace=BCS_BoneSpace
    alphaBoolBlend {blendInTime=0.1, blendOption=Linear, blendOutTime=0.1}
    boneToModify {boneName=dyn_lowerarm_sleeve_l_MSTR}
    rotation {roll=30}
   }
   n3 = RigidBody!(
    ComponentPose=@n2
    Alpha
    OverrideWorldGravity:Vector="0,0,-980"
    ExternalForce
    SimSpaceSettings:Sim Space Settings Structure=(WorldAlpha=0.8,VelocityScaleZ=0.2,DampingAlpha=0,MaxLinearVelocity=100,MaxAngularVelocity=5,MaxLinearAcceleration=1000,MaxAngularAcceleration=30,ExternalLinearDragV=(X=0.5,Y=0.5,Z=0.5),ExternalLinearVelocity=(X=0,Y=0,Z=0),ExternalAngularVelocity=(X=0,Y=0,Z=0))
   ) -> (Pose) {
    bOverrideWorldGravity=true
    overridePhysicsAsset=&/M_MED_ClawPad_HostFull_LeftSleeve_Physics
    simulationSpace=BaseBoneSpace
    alphaBoolBlend {blendOption=Linear}
    baseBoneRef {boneName=lowerarm_l}
    componentAppliedLinearAccClamp=<0 0 0>
    overrideWorldGravity {z=-980}
    simSpaceSettings {
     dampingAlpha=0
     maxAngularAcceleration=30
     maxAngularVelocity=5
     maxLinearAcceleration=1000
     maxLinearVelocity=100
     velocityScaleZ=0.2
     worldAlpha=0.8
     externalLinearDragV=<0.5 0.5 0.5>
    }
   }
   n4 = RigidBody(
    ComponentPose=@n3
    OverrideWorldGravity="0,0,-980"
    SimSpaceSettings=(WorldAlpha=0.8,VelocityScaleZ=0.2,DampingAlpha=0,MaxLinearVelocity=100,MaxAngularVelocity=5,MaxLinearAcceleration=1000,MaxAngularAcceleration=30,ExternalLinearDragV=(X=0.5,Y=0.5,Z=0.5),ExternalLinearVelocity=(X=0,Y=0,Z=0),ExternalAngularVelocity=(X=0,Y=0,Z=0))
   ) {
    bOverrideWorldGravity=true
    overridePhysicsAsset=&/M_MED_ClawPad_HostFull_RightSleeve_Physics
    simulationSpace=BaseBoneSpace
    alphaBoolBlend {blendOption=Linear}
    baseBoneRef {boneName=lowerarm_r}
    componentAppliedLinearAccClamp=<0 0 0>
    overrideWorldGravity {z=-980}
    simSpaceSettings {
     dampingAlpha=0
     maxAngularAcceleration=30
     maxAngularVelocity=5
     maxLinearAcceleration=1000
     maxLinearVelocity=100
     velocityScaleZ=0.2
     worldAlpha=0.8
     externalLinearDragV=<0.5 0.5 0.5>
    }
   }
   n5 = Anim Dynamics - Bone: dyn_lowerarm_l_sleeve_twist_01(ComponentPose=Local To Component(LocalPose=Use cached pose 'Post Slope Posing'()), Alpha=1, ExternalForce="0,0,0") {
    angularDampingOverride=0.95
    angularSpringConstant=50
    bAngularSpring=true
    bOverrideAngularDamping=true
    bUseGravityOverride=true
    gravityScale=0.35
    alphaBoolBlend {blendOption=Linear}
    boundBone {boneName=dyn_lowerarm_l_sleeve_twist_01}
    componentAppliedLinearAccClamp=<200 200 200>
    componentLinearAccScale=<1 1 1>
    componentLinearVelScale=<1 1 1>
    gravityOverride {z=-1200}
    physicsBodyDefinitions [{
      boundBone {boneName=dyn_lowerarm_l_sleeve_twist_01}
      boxExtents=<20 20 40>
      constraintSetup {angularConstraintType=Cone, coneAngle=5, angularLimitsMax=<15 2 2>, angularLimitsMin=<-15 -2 -2>, angularTarget=<0 0 1>}
      localJointOffset=<4 0 -10>
     }]
   }
   n6 = Anim Dynamics - Bone: dyn_lowerarm_l_sleeve_twist_01_end(ComponentPose=@n5, Alpha=1, ExternalForce="0,0,0") {
    angularDampingOverride=0.95
    angularSpringConstant=50
    bAngularSpring=true
    bOverrideAngularDamping=true
    bUseGravityOverride=true
    gravityScale=0.35
    alphaBoolBlend {blendOption=Linear}
    boundBone {boneName=dyn_lowerarm_l_sleeve_twist_01_end}
    componentAppliedLinearAccClamp=<200 200 200>
    componentLinearAccScale=<1 1 1>
    componentLinearVelScale=<1 1 1>
    gravityOverride {z=-1200}
    physicsBodyDefinitions [{
      boundBone {boneName=dyn_lowerarm_l_sleeve_twist_01_end}
      boxExtents=<20 20 40>
      constraintSetup {angularConstraintType=Cone, coneAngle=5, angularLimitsMax=<50 42 -40>, angularLimitsMin=<-50 0 40>, angularTarget=<0 0 1>}
      localJointOffset=<4 0 -10>
     }]
   }
   n7 = Anim Dynamics - Bone: dyn_lowerarm_l_sleeve_twist_02(ComponentPose=@n6, Alpha=1, ExternalForce="0,0,0") {
    angularDampingOverride=0.95
    angularSpringConstant=50
    bAngularSpring=true
    bOverrideAngularDamping=true
    bUseGravityOverride=true
    gravityScale=0.35
    alphaBoolBlend {blendOption=Linear}
    boundBone {boneName=dyn_lowerarm_l_sleeve_twist_02}
    componentAppliedLinearAccClamp=<200 200 200>
    componentLinearAccScale=<1 1 1>
    componentLinearVelScale=<1 1 1>
    gravityOverride {z=-1200}
    physicsBodyDefinitions [{
      boundBone {boneName=dyn_lowerarm_l_sleeve_twist_02}
      boxExtents=<20 20 40>
      constraintSetup {angularConstraintType=Cone, coneAngle=5, angularLimitsMax=<30 2 2>, angularLimitsMin=<-30 -2 -2>, angularTarget=<0 0 1>}
      localJointOffset=<4 0 -10>
     }]
   }
   n8 = Anim Dynamics - Bone: dyn_lowerarm_l_sleeve_twist_02_end(ComponentPose=@n7, Alpha=1, ExternalForce="0,0,0") {
    angularDampingOverride=0.95
    angularSpringConstant=50
    bAngularSpring=true
    bOverrideAngularDamping=true
    bUseGravityOverride=true
    gravityScale=0.35
    alphaBoolBlend {blendOption=Linear}
    boundBone {boneName=dyn_lowerarm_l_sleeve_twist_02_end}
    componentAppliedLinearAccClamp=<200 200 200>
    componentLinearAccScale=<1 1 1>
    componentLinearVelScale=<1 1 1>
    gravityOverride {z=-1200}
    physicsBodyDefinitions [{
      boundBone {boneName=dyn_lowerarm_l_sleeve_twist_02_end}
      boxExtents=<20 20 40>
      constraintSetup {angularConstraintType=Cone, coneAngle=5, angularLimitsMax=<50 42 -40>, angularLimitsMin=<-50 0 40>, angularTarget=<0 0 1>}
      localJointOffset=<4 0 -10>
     }]
   }
   n9 = Anim Dynamics - Bone: dyn_lowerarm_l_sleeve_twist_03(ComponentPose=@n8, Alpha=1, ExternalForce="0,0,0") {
    angularDampingOverride=0.95
    angularSpringConstant=50
    bAngularSpring=true
    bOverrideAngularDamping=true
    bUseGravityOverride=true
    gravityScale=0.35
    alphaBoolBlend {blendOption=Linear}
    boundBone {boneName=dyn_lowerarm_l_sleeve_twist_03}
    componentAppliedLinearAccClamp=<200 200 200>
    componentLinearAccScale=<1 1 1>
    componentLinearVelScale=<1 1 1>
    gravityOverride {z=-1200}
    physicsBodyDefinitions [{
      boundBone {boneName=dyn_lowerarm_l_sleeve_twist_03}
      boxExtents=<20 20 40>
      constraintSetup {angularConstraintType=Cone, coneAngle=5, angularLimitsMax=<30 5 0>, angularLimitsMin=<-30 -5 0>, angularTarget=<0 0 1>}
      localJointOffset=<4 0 -10>
     }]
   }
   n10 = Anim Dynamics - Bone: dyn_lowerarm_l_sleeve_twist_03_end(ComponentPose=@n9, Alpha=1, ExternalForce="0,0,0") {
    angularDampingOverride=0.95
    angularSpringConstant=50
    bAngularSpring=true
    bOverrideAngularDamping=true
    bUseGravityOverride=true
    gravityScale=0.35
    alphaBoolBlend {blendOption=Linear}
    boundBone {boneName=dyn_lowerarm_l_sleeve_twist_03_end}
    componentAppliedLinearAccClamp=<200 200 200>
    componentLinearAccScale=<1 1 1>
    componentLinearVelScale=<1 1 1>
    gravityOverride {z=-1200}
    physicsBodyDefinitions [{
      boundBone {boneName=dyn_lowerarm_l_sleeve_twist_03_end}
      boxExtents=<20 20 40>
      constraintSetup {angularConstraintType=Cone, coneAngle=5, angularLimitsMax=<50 42 -40>, angularLimitsMin=<-50 0 40>, angularTarget=<0 0 1>}
      localJointOffset=<4 0 -10>
     }]
   }
   n11 = Anim Dynamics - Bone: dyn_lowerarm_r_sleeve_twist_01(ComponentPose=@n10, Alpha=1, ExternalForce="0,0,0") {
    angularDampingOverride=0.95
    angularSpringConstant=50
    bAngularSpring=true
    bOverrideAngularDamping=true
    bUseGravityOverride=true
    gravityScale=0.35
    alphaBoolBlend {blendOption=Linear}
    boundBone {boneName=dyn_lowerarm_r_sleeve_twist_01}
    componentAppliedLinearAccClamp=<200 200 200>
    componentLinearAccScale=<1 1 1>
    componentLinearVelScale=<1 1 1>
    gravityOverride {z=-1200}
    physicsBodyDefinitions [{
      boundBone {boneName=dyn_lowerarm_r_sleeve_twist_01}
      boxExtents=<20 20 40>
      constraintSetup {angularConstraintType=Cone, coneAngle=5, angularLimitsMax=<15 2 2>, angularLimitsMin=<-15 -2 -2>, angularTarget=<0 0 1>}
      localJointOffset=<-4 0 10>
     }]
   }
   n12 = Anim Dynamics - Bone: dyn_lowerarm_r_sleeve_twist_01_end(ComponentPose=@n11, Alpha=1, ExternalForce="0,0,0") {
    angularDampingOverride=0.95
    angularSpringConstant=50
    bAngularSpring=true
    bOverrideAngularDamping=true
    bUseGravityOverride=true
    gravityScale=0.35
    alphaBoolBlend {blendOption=Linear}
    boundBone {boneName=dyn_lowerarm_r_sleeve_twist_01_end}
    componentAppliedLinearAccClamp=<200 200 200>
    componentLinearAccScale=<1 1 1>
    componentLinearVelScale=<1 1 1>
    gravityOverride {z=-1200}
    physicsBodyDefinitions [{
      boundBone {boneName=dyn_lowerarm_r_sleeve_twist_01_end}
      boxExtents=<20 20 40>
      constraintSetup {angularConstraintType=Cone, coneAngle=5, angularLimitsMax=<50 42 -40>, angularLimitsMin=<-50 0 40>, angularTarget=<0 0 1>}
      localJointOffset=<-4 0 10>
     }]
   }
   n13 = Anim Dynamics - Bone: dyn_lowerarm_r_sleeve_twist_02(ComponentPose=@n12, Alpha=1, ExternalForce="0,0,0") {
    angularDampingOverride=0.95
    angularSpringConstant=50
    bAngularSpring=true
    bOverrideAngularDamping=true
    bUseGravityOverride=true
    gravityScale=0.35
    alphaBoolBlend {blendOption=Linear}
    boundBone {boneName=dyn_lowerarm_r_sleeve_twist_02}
    componentAppliedLinearAccClamp=<200 200 200>
    componentLinearAccScale=<1 1 1>
    componentLinearVelScale=<1 1 1>
    gravityOverride {z=-1200}
    physicsBodyDefinitions [{
      boundBone {boneName=dyn_lowerarm_r_sleeve_twist_02}
      boxExtents=<20 20 40>
      constraintSetup {angularConstraintType=Cone, coneAngle=5, angularLimitsMax=<30 2 2>, angularLimitsMin=<-30 -2 -2>, angularTarget=<0 0 1>}
      localJointOffset=<-4 0 10>
     }]
   }
   n14 = Anim Dynamics - Bone: dyn_lowerarm_r_sleeve_twist_02_end(ComponentPose=@n13, Alpha=1, ExternalForce="0,0,0") {
    angularDampingOverride=0.95
    angularSpringConstant=50
    bAngularSpring=true
    bOverrideAngularDamping=true
    bUseGravityOverride=true
    gravityScale=0.35
    alphaBoolBlend {blendOption=Linear}
    boundBone {boneName=dyn_lowerarm_r_sleeve_twist_02_end}
    componentAppliedLinearAccClamp=<200 200 200>
    componentLinearAccScale=<1 1 1>
    componentLinearVelScale=<1 1 1>
    gravityOverride {z=-1200}
    physicsBodyDefinitions [{
      boundBone {boneName=dyn_lowerarm_r_sleeve_twist_02_end}
      boxExtents=<20 20 40>
      constraintSetup {angularConstraintType=Cone, coneAngle=5, angularLimitsMax=<50 42 -40>, angularLimitsMin=<-50 0 40>, angularTarget=<0 0 1>}
      localJointOffset=<-4 0 10>
     }]
   }
   n15 = Anim Dynamics - Bone: dyn_lowerarm_r_sleeve_twist_03(ComponentPose=@n14, Alpha=1, ExternalForce="0,0,0") {
    angularDampingOverride=0.95
    angularSpringConstant=50
    bAngularSpring=true
    bOverrideAngularDamping=true
    bUseGravityOverride=true
    gravityScale=0.35
    alphaBoolBlend {blendOption=Linear}
    boundBone {boneName=dyn_lowerarm_r_sleeve_twist_03}
    componentAppliedLinearAccClamp=<200 200 200>
    componentLinearAccScale=<1 1 1>
    componentLinearVelScale=<1 1 1>
    gravityOverride {z=-1200}
    physicsBodyDefinitions [{
      boundBone {boneName=dyn_lowerarm_r_sleeve_twist_03}
      boxExtents=<20 20 40>
      constraintSetup {angularConstraintType=Cone, coneAngle=5, angularLimitsMax=<30 5 0>, angularLimitsMin=<-30 -5 0>, angularTarget=<0 0 1>}
      localJointOffset=<-4 0 10>
     }]
   }
   n16 = Anim Dynamics - Bone: dyn_lowerarm_r_sleeve_twist_03_end(ComponentPose=@n15, Alpha=1, ExternalForce="0,0,0") {
    angularDampingOverride=0.95
    angularSpringConstant=50
    bAngularSpring=true
    bOverrideAngularDamping=true
    bUseGravityOverride=true
    gravityScale=0.35
    alphaBoolBlend {blendOption=Linear}
    boundBone {boneName=dyn_lowerarm_r_sleeve_twist_03_end}
    componentAppliedLinearAccClamp=<200 200 200>
    componentLinearAccScale=<1 1 1>
    componentLinearVelScale=<1 1 1>
    gravityOverride {z=-1200}
    physicsBodyDefinitions [{
      boundBone {boneName=dyn_lowerarm_r_sleeve_twist_03_end}
      boxExtents=<20 20 40>
      constraintSetup {angularConstraintType=Cone, coneAngle=5, angularLimitsMax=<50 -20 40>, angularLimitsMin=<-50 20 -40>, angularTarget=<0 0 1>}
      localJointOffset=<-4 0 10>
     }]
   }
   n17 = "Transform (Modify) Bone - Bone: dyn_front_skirt_l_MSTR"(
    ComponentPose=Local To Component(LocalPose=Use cached pose 'Post Slope Posing'())
    Alpha=1
    Translation="0,0,0"
    Rotation="-20,-20,0"
    Scale="1,1,1"
   ) {
    rotationMode=BMM_Additive
    rotationSpace=BCS_BoneSpace
    translationMode=BMM_Additive
    translationSpace=BCS_BoneSpace
    alphaBoolBlend {blendOption=Linear}
    boneToModify {boneName=dyn_front_skirt_l_MSTR}
    rotation {pitch=-20, yaw=-20}
   }
   n18 = "Transform (Modify) Bone - Bone: dyn_front_skirt_r_MSTR"(ComponentPose=@n17, Alpha=1, Translation="0,0,0", Rotation="0,-20,0", Scale="1,1,1") {
    rotationMode=BMM_Additive
    rotationSpace=BCS_BoneSpace
    translationMode=BMM_Additive
    translationSpace=BCS_BoneSpace
    alphaBoolBlend {blendOption=Linear}
    boneToModify {boneName=dyn_front_skirt_r_MSTR}
    rotation {yaw=-20}
   }
   n19 = RigidBody(
    ComponentPose=@n18
    OverrideWorldGravity="0,0,-980"
    SimSpaceSettings=(WorldAlpha=1,VelocityScaleZ=0.4,DampingAlpha=0,MaxLinearVelocity=400,MaxAngularVelocity=7.5,MaxLinearAcceleration=3000,MaxAngularAcceleration=30,ExternalLinearDragV=(X=0.5,Y=0.5,Z=0.2),ExternalLinearVelocity=(X=0,Y=0,Z=0),ExternalAngularVelocity=(X=0,Y=0,Z=0))
   ) {
    bOverrideWorldGravity=true
    overridePhysicsAsset=&/M_MED_ClawPad_HostFull_Robe_Physics
    simulationSpace=BaseBoneSpace
    alphaBoolBlend {blendOption=Linear}
    baseBoneRef {boneName=dyn_skirt_SIMSPACE}
    componentAppliedLinearAccClamp=<0 0 0>
    overrideWorldGravity {z=-980}
    simSpaceSettings {
     dampingAlpha=0
     maxAngularAcceleration=30
     maxAngularVelocity=7.5
     maxLinearAcceleration=3000
     maxLinearVelocity=400
     velocityScaleZ=0.4
     worldAlpha=1
     externalLinearDragV=<0.5 0.5 0.2>
    }
   }
   n20 = Layered blend per bone!(
    BasePose=Use cached pose 'Post Slope Posing'()
    BlendPoses_0=Blend Poses by bool(
     BlendPose_0=Component To Local(ComponentPose=@n16)
     BlendPose_1=Component To Local(ComponentPose=@n4)
     BlendTime_0=0
     BlendTime_1=0
     bActiveValue=Get UseAnimDynamicSleeves() -> (UseAnimDynamicSleeves:bool)
    ) {blendType=Linear}
    BlendPoses_1:Pose Link Structure=Component To Local(ComponentPose=@n19)
    BlendWeights_0
    BlendWeights_1:float=1
   ) -> (Pose) {
    blendPoses [{linkId=-1, sourceLinkId=-1}, {linkId=-1, sourceLinkId=-1}]
    blendWeights [1,1]
    layerSetup [{branchFilters [{boneName=dyn_lowerarm_sleeve_l_MSTR}, {boneName=dyn_lowerarm_sleeve_r_MSTR}]}, {branchFilters [{boneName=dyn_skirt_MSTR}]}]
   }
   n21 = Control Rig!(
    Source=@n20
    Alpha
    player speed:double=Get PawnSpeed2D()
    rotation speed:double=Get!() -> (PawnRotationRate_Roll:float, PawnRotationRate_Pitch:float, PawnRotationRate_Yaw:float).PawnRotationRate_Yaw
    IsHostFull:bool=true
    IsPlayingEmote:bool=Get bIsPlayingEmote()
    IsBaseballSliding:bool=Get bIsBaseballSliding()
    IsCustomLobbyIdle:double=Get Custom Idle Is Playing Float() -> (Custom Idle Is Playing Float:float)
    IsInFrontend:bool=Get bIsInFrontEnd()
   ) -> (Pose) {
    alphaBoolBlend {blendOption=Linear}
    controlRigAssetReference {blueprintRigClass=&/M_MED_ClawPad_HostFull_PostSim_CtrlRig.M_MED_ClawPad_HostFull_PostSim_CtrlRig_C}
   }
   Save cached pose 'Post RBAN'(Pose=@n21) {cacheName=Post RBAN}
   n22 = Anim Dynamics - Bone: dyn_front_panel_01!(
    SimulationSpace:AnimPhysSimSpaceType Enum=Component
    ComponentPose=Local To Component(LocalPose=Use cached pose 'Post RBAN'())
    Alpha=1
    ExternalForce="0,0,0"
   ) -> (Pose) {
    angularDampingOverride=1
    angularSpringConstant=10000
    bAngularSpring=true
    bOverrideAngularDamping=true
    bUseGravityOverride=true
    gravityScale=0.35
    simulationSpace=RootRelative
    alphaBoolBlend {blendOption=Linear}
    boundBone {boneName=dyn_front_panel_01}
    componentAppliedLinearAccClamp=<1200 1200 1200>
    componentLinearAccScale=<1 1 1>
    componentLinearVelScale=<1 1 1>
    gravityOverride {z=-980}
    physicsBodyDefinitions [{
      boundBone {boneName=dyn_front_panel_01}
      boxExtents=<20 30 30>
      constraintSetup {angularLimitsMax=<5 5 0>, angularLimitsMin=<-5 -5 -20>, angularTarget=<1 0 0>}
      localJointOffset=<4 0 0>
     }]
   }
   n23 = Anim Dynamics - Bone: dyn_front_panel_02!(SimulationSpace:AnimPhysSimSpaceType Enum, ComponentPose=@n22, Alpha=1, ExternalForce="0,0,0") -> (Pose) {
    angularDampingOverride=1
    angularSpringConstant=10000
    bAngularSpring=true
    bOverrideAngularDamping=true
    bUseGravityOverride=true
    gravityScale=0.35
    simulationSpace=RootRelative
    alphaBoolBlend {blendOption=Linear}
    boundBone {boneName=dyn_front_panel_02}
    componentAppliedLinearAccClamp=<1200 1200 1200>
    componentLinearAccScale=<1 1 1>
    componentLinearVelScale=<1 1 1>
    gravityOverride {z=-980}
    physicsBodyDefinitions [{
      boundBone {boneName=dyn_front_panel_02}
      boxExtents=<20 30 30>
      constraintSetup {angularLimitsMax=<5 5 0>, angularLimitsMin=<-5 -5 -20>, angularTarget=<1 0 0>}
      localJointOffset=<4 0 0>
     }]
   }
   n24 = Anim Dynamics - Bone: dyn_front_panel_03!(SimulationSpace:AnimPhysSimSpaceType Enum, ComponentPose=@n23, Alpha=1, ExternalForce="0,0,0") -> (Pose) {
    angularDampingOverride=1
    angularSpringConstant=10000
    bAngularSpring=true
    bOverrideAngularDamping=true
    bUseGravityOverride=true
    gravityScale=0.35
    simulationSpace=RootRelative
    alphaBoolBlend {blendOption=Linear}
    boundBone {boneName=dyn_front_panel_03}
    componentAppliedLinearAccClamp=<1200 1200 1200>
    componentLinearAccScale=<1 1 1>
    componentLinearVelScale=<1 1 1>
    gravityOverride {z=-980}
    physicsBodyDefinitions [{
      boundBone {boneName=dyn_front_panel_03}
      boxExtents=<20 30 30>
      constraintSetup {angularLimitsMax=<5 5 0>, angularLimitsMin=<-5 -5 -20>, angularTarget=<1 0 0>}
      localJointOffset=<4 0 0>
     }]
   }
   n25 = Anim Dynamics - Bone: dyn_front_panel_04!(SimulationSpace:AnimPhysSimSpaceType Enum, ComponentPose=@n24, Alpha=1, ExternalForce="0,0,0") -> (Pose) {
    angularDampingOverride=1
    angularSpringConstant=10000
    bAngularSpring=true
    bOverrideAngularDamping=true
    bUseGravityOverride=true
    gravityScale=0.35
    simulationSpace=RootRelative
    alphaBoolBlend {blendOption=Linear}
    boundBone {boneName=dyn_front_panel_04}
    componentAppliedLinearAccClamp=<1200 1200 1200>
    componentLinearAccScale=<1 1 1>
    componentLinearVelScale=<1 1 1>
    gravityOverride {z=-980}
    physicsBodyDefinitions [{
      boundBone {boneName=dyn_front_panel_04}
      boxExtents=<20 30 30>
      constraintSetup {angularLimitsMax=<5 5 0>, angularLimitsMin=<-5 -5 -20>, angularTarget=<1 0 0>}
      localJointOffset=<4 0 0>
     }]
   }
   n26 = Anim Dynamics - Bone: dyn_front_panel_05!(SimulationSpace:AnimPhysSimSpaceType Enum, ComponentPose=@n25, Alpha=1, ExternalForce="0,0,0") -> (Pose) {
    angularDampingOverride=1
    angularSpringConstant=10000
    bAngularSpring=true
    bOverrideAngularDamping=true
    bUseGravityOverride=true
    gravityScale=0.35
    simulationSpace=RootRelative
    alphaBoolBlend {blendOption=Linear}
    boundBone {boneName=dyn_front_panel_05}
    componentAppliedLinearAccClamp=<1200 1200 1200>
    componentLinearAccScale=<1 1 1>
    componentLinearVelScale=<1 1 1>
    gravityOverride {z=-980}
    physicsBodyDefinitions [{
      boundBone {boneName=dyn_front_panel_05}
      boxExtents=<20 30 30>
      constraintSetup {angularLimitsMax=<5 5 0>, angularLimitsMin=<-5 -5 -20>, angularTarget=<1 0 0>}
      localJointOffset=<4 0 0>
     }]
   }
   n27 = Anim Dynamics - Bone: dyn_lowerarm_l_sleeve_twist_03!(SimulationSpace:AnimPhysSimSpaceType Enum, ComponentPose=@n26, Alpha=0.8, ExternalForce="0,0,0") -> (Pose) {
    angularDampingOverride=1
    angularSpringConstant=20000
    bLinearSpring=true
    bOverrideLinearDamping=true
    bUseGravityOverride=true
    bUsePlanarLimit=false
    gravityScale=0.35
    linearDampingOverride=1
    linearSpringConstant=1
    simulationSpace=RootRelative
    alphaBoolBlend {blendOption=Linear}
    boundBone {boneName=dyn_lowerarm_l_sleeve_twist_03}
    componentAppliedLinearAccClamp=<500 500 500>
    componentLinearAccScale=<1 1 1>
    componentLinearVelScale=<1 1 1>
    gravityOverride {z=-980}
    physicsBodyDefinitions [{
      boundBone {boneName=dyn_lowerarm_l_sleeve_twist_03}
      boxExtents=<50 15 15>
      constraintSetup {angularLimitsMax=<0 0 0>, angularLimitsMin=<0 0 0>, angularTarget=<1 0 0>, linearAxesMin=<-3 0 0>}
      localJointOffset=<5 0 0>
     }]
    relativeSpaceBone {boneName=lowerarm_l}
   }
   n28 = Anim Dynamics - Bone: dyn_lowerarm_l_sleeve_twist_02!(SimulationSpace:AnimPhysSimSpaceType Enum, ComponentPose=@n27, Alpha=0.8, ExternalForce="0,0,0") -> (Pose) {
    angularDampingOverride=1
    angularSpringConstant=20000
    bLinearSpring=true
    bOverrideLinearDamping=true
    bUseGravityOverride=true
    bUsePlanarLimit=false
    gravityScale=0.35
    linearDampingOverride=1
    linearSpringConstant=1
    simulationSpace=RootRelative
    alphaBoolBlend {blendOption=Linear}
    boundBone {boneName=dyn_lowerarm_l_sleeve_twist_02}
    componentAppliedLinearAccClamp=<500 500 500>
    componentLinearAccScale=<1 1 1>
    componentLinearVelScale=<1 1 1>
    gravityOverride {z=-980}
    physicsBodyDefinitions [{
      boundBone {boneName=dyn_lowerarm_l_sleeve_twist_02}
      boxExtents=<50 15 15>
      constraintSetup {angularLimitsMax=<0 0 0>, angularLimitsMin=<0 0 0>, angularTarget=<1 0 0>, linearAxesMin=<-3 0 0>}
      localJointOffset=<5 0 0>
     }]
    relativeSpaceBone {boneName=lowerarm_l}
   }
   n29 = Anim Dynamics - Bone: dyn_lowerarm_r_sleeve_twist_03!(SimulationSpace:AnimPhysSimSpaceType Enum, ComponentPose=@n28, Alpha=0.8, ExternalForce="0,0,0") -> (Pose) {
    angularDampingOverride=1
    angularSpringConstant=20000
    bLinearSpring=true
    bOverrideLinearDamping=true
    bUseGravityOverride=true
    bUsePlanarLimit=false
    gravityScale=0.35
    linearDampingOverride=1
    linearSpringConstant=1
    simulationSpace=RootRelative
    alphaBoolBlend {blendOption=Linear}
    boundBone {boneName=dyn_lowerarm_r_sleeve_twist_03}
    componentAppliedLinearAccClamp=<500 500 500>
    componentLinearAccScale=<1 1 1>
    componentLinearVelScale=<1 1 1>
    gravityOverride {z=-980}
    physicsBodyDefinitions [{
      boundBone {boneName=dyn_lowerarm_r_sleeve_twist_03}
      boxExtents=<50 15 15>
      constraintSetup {angularLimitsMax=<0 0 0>, angularLimitsMin=<0 0 0>, angularTarget=<1 0 0>, linearAxesMax=<3 0 0>}
      localJointOffset=<-5 0 0>
     }]
   }
   n30 = Anim Dynamics - Bone: dyn_lowerarm_r_sleeve_twist_02!(SimulationSpace:AnimPhysSimSpaceType Enum, ComponentPose=@n29, Alpha=0.8, ExternalForce="0,0,0") -> (Pose) {
    angularDampingOverride=1
    angularSpringConstant=20000
    bLinearSpring=true
    bOverrideLinearDamping=true
    bUseGravityOverride=true
    bUsePlanarLimit=false
    gravityScale=0.35
    linearDampingOverride=1
    linearSpringConstant=1
    simulationSpace=RootRelative
    alphaBoolBlend {blendOption=Linear}
    boundBone {boneName=dyn_lowerarm_r_sleeve_twist_02}
    componentAppliedLinearAccClamp=<500 500 500>
    componentLinearAccScale=<1 1 1>
    componentLinearVelScale=<1 1 1>
    gravityOverride {z=-980}
    physicsBodyDefinitions [{
      boundBone {boneName=dyn_lowerarm_r_sleeve_twist_02}
      boxExtents=<50 15 15>
      constraintSetup {angularLimitsMax=<0 0 0>, angularLimitsMin=<0 0 0>, angularTarget=<1 0 0>, linearAxesMax=<3 0 0>}
      localJointOffset=<-5 0 0>
     }]
   }
   n31 = Anim Dynamics - Bone: dyn_shawl_shoulder_l_01!(SimulationSpace:AnimPhysSimSpaceType Enum, ComponentPose=@n30, Alpha=0.8, ExternalForce="0,0,0") -> (Pose) {
    angularDampingOverride=0.995
    angularSpringConstant=10000
    bAngularSpring=true
    bOverrideAngularDamping=true
    bUseGravityOverride=true
    gravityScale=0.35
    simulationSpace=RootRelative
    alphaBoolBlend {blendOption=Linear}
    boundBone {boneName=dyn_shawl_shoulder_l_01}
    componentAppliedLinearAccClamp=<1000 1000 1300>
    componentLinearAccScale=<0.5 0.5 1>
    componentLinearVelScale=<0.5 0.5 1>
    gravityOverride {z=-1500}
    physicsBodyDefinitions [{
      boundBone {boneName=dyn_shawl_shoulder_l_01}
      boxExtents=<20 35 35>
      constraintSetup {angularLimitsMax=<5 5 2>, angularLimitsMin=<-5 -5 -15>, angularTarget=<1 0 0>}
      localJointOffset=<4 0 0>
     }]
   }
   n32 = Anim Dynamics - Bone: dyn_shawl_shoulder_l_02!(SimulationSpace:AnimPhysSimSpaceType Enum, ComponentPose=@n31, Alpha=0.8, ExternalForce="0,0,0") -> (Pose) {
    angularDampingOverride=0.995
    angularSpringConstant=10000
    bAngularSpring=true
    bOverrideAngularDamping=true
    bUseGravityOverride=true
    gravityScale=0.35
    simulationSpace=RootRelative
    alphaBoolBlend {blendOption=Linear}
    boundBone {boneName=dyn_shawl_shoulder_l_02}
    componentAppliedLinearAccClamp=<1000 1000 1300>
    componentLinearAccScale=<0.5 0.5 1>
    componentLinearVelScale=<0.5 0.5 1>
    gravityOverride {z=-1500}
    physicsBodyDefinitions [{
      boundBone {boneName=dyn_shawl_shoulder_l_02}
      boxExtents=<20 35 35>
      constraintSetup {angularLimitsMax=<5 5 2>, angularLimitsMin=<-5 -5 -25>, angularTarget=<1 0 0>}
      localJointOffset=<4 0 0>
     }]
   }
   n33 = Anim Dynamics - Bone: dyn_shawl_shoulder_l_03!(SimulationSpace:AnimPhysSimSpaceType Enum, ComponentPose=@n32, Alpha=0.8, ExternalForce="0,0,0") -> (Pose) {
    angularDampingOverride=0.995
    angularSpringConstant=10000
    bAngularSpring=true
    bOverrideAngularDamping=true
    bUseGravityOverride=true
    gravityScale=0.35
    simulationSpace=RootRelative
    alphaBoolBlend {blendOption=Linear}
    boundBone {boneName=dyn_shawl_shoulder_l_03}
    componentAppliedLinearAccClamp=<1000 1000 1300>
    componentLinearAccScale=<0.5 0.5 1>
    componentLinearVelScale=<0.5 0.5 1>
    gravityOverride {z=-1500}
    physicsBodyDefinitions [{
      boundBone {boneName=dyn_shawl_shoulder_l_03}
      boxExtents=<20 35 35>
      constraintSetup {angularLimitsMax=<12.5 12.5 2>, angularLimitsMin=<-12.5 -12.5 -25>, angularTarget=<1 0 0>}
      localJointOffset=<4 0 0>
     }]
   }
   n34 = Anim Dynamics - Bone: dyn_shawl_reardelt_l_01!(SimulationSpace:AnimPhysSimSpaceType Enum, ComponentPose=@n33, Alpha=0.8, ExternalForce="0,0,0") -> (Pose) {
    angularDampingOverride=0.995
    angularSpringConstant=9000
    bAngularSpring=true
    bOverrideAngularDamping=true
    bUseGravityOverride=true
    gravityScale=0.35
    simulationSpace=RootRelative
    alphaBoolBlend {blendOption=Linear}
    boundBone {boneName=dyn_shawl_reardelt_l_01}
    componentAppliedLinearAccClamp=<1000 1000 1300>
    componentLinearAccScale=<0.5 0.5 1>
    componentLinearVelScale=<0.5 0.5 1>
    gravityOverride {z=-1500}
    physicsBodyDefinitions [{
      boundBone {boneName=dyn_shawl_reardelt_l_01}
      boxExtents=<20 35 35>
      constraintSetup {angularLimitsMax=<5 5 2>, angularLimitsMin=<-5 -5 -15>, angularTarget=<1 0 0>}
      localJointOffset=<4 -0.5 0.5>
     }]
   }
   n35 = Anim Dynamics - Bone: dyn_shawl_reardelt_l_02!(SimulationSpace:AnimPhysSimSpaceType Enum, ComponentPose=@n34, Alpha=0.8, ExternalForce="0,0,0") -> (Pose) {
    angularDampingOverride=0.995
    angularSpringConstant=10000
    bAngularSpring=true
    bOverrideAngularDamping=true
    bUseGravityOverride=true
    gravityScale=0.35
    simulationSpace=RootRelative
    alphaBoolBlend {blendOption=Linear}
    boundBone {boneName=dyn_shawl_reardelt_l_02}
    componentAppliedLinearAccClamp=<1000 1000 1300>
    componentLinearAccScale=<0.5 0.5 1>
    componentLinearVelScale=<0.5 0.5 1>
    gravityOverride {z=-1500}
    physicsBodyDefinitions [{
      boundBone {boneName=dyn_shawl_reardelt_l_02}
      boxExtents=<20 35 35>
      constraintSetup {angularLimitsMax=<12.5 12.5 2>, angularLimitsMin=<-12.5 -12.5 -25>, angularTarget=<1 0 0>}
      localJointOffset=<4 0 0>
     }]
   }
   n36 = Anim Dynamics - Bone: dyn_shawl_scapula_l_01!(SimulationSpace:AnimPhysSimSpaceType Enum, ComponentPose=@n35, Alpha=0.8, ExternalForce="0,0,0") -> (Pose) {
    angularDampingOverride=0.995
    angularSpringConstant=12000
    bAngularSpring=true
    bOverrideAngularDamping=true
    bUseGravityOverride=true
    gravityScale=0.35
    simulationSpace=RootRelative
    alphaBoolBlend {blendOption=Linear}
    boundBone {boneName=dyn_shawl_scapula_l_01}
    componentAppliedLinearAccClamp=<1000 1000 1300>
    componentLinearAccScale=<0.5 0.5 1>
    componentLinearVelScale=<0.5 0.5 1>
    gravityOverride {z=-1500}
    physicsBodyDefinitions [{
      boundBone {boneName=dyn_shawl_scapula_l_01}
      boxExtents=<20 35 35>
      constraintSetup {angularLimitsMax=<5 15 0>, angularLimitsMin=<-5 -15 -30>, angularTarget=<1 0 0>}
      localJointOffset=<4 0.5 -0.5>
     }]
   }
   n37 = Anim Dynamics - Bone: dyn_shawl_scapula_l_02!(SimulationSpace:AnimPhysSimSpaceType Enum, ComponentPose=@n36, Alpha=0.8, ExternalForce="0,0,0") -> (Pose) {
    angularDampingOverride=0.995
    angularSpringConstant=10000
    bAngularSpring=true
    bOverrideAngularDamping=true
    bUseGravityOverride=true
    gravityScale=0.35
    simulationSpace=RootRelative
    alphaBoolBlend {blendOption=Linear}
    boundBone {boneName=dyn_shawl_scapula_l_02}
    componentAppliedLinearAccClamp=<1000 1000 1300>
    componentLinearAccScale=<0.5 0.5 1>
    componentLinearVelScale=<0.5 0.5 1>
    gravityOverride {z=-1500}
    physicsBodyDefinitions [{
      boundBone {boneName=dyn_shawl_scapula_l_02}
      boxExtents=<20 35 35>
      constraintSetup {angularLimitsMax=<12.5 15 0>, angularLimitsMin=<-12.5 -15 -25>, angularTarget=<1 0 0>}
      localJointOffset=<4 0 0>
     }]
   }
   n38 = Anim Dynamics - Bone: dyn_shawl_frontdelt_l_01!(SimulationSpace:AnimPhysSimSpaceType Enum, ComponentPose=@n37, Alpha=0.5, ExternalForce="0,0,0") -> (Pose) {
    angularDampingOverride=0.995
    angularSpringConstant=10000
    bAngularSpring=true
    bOverrideAngularDamping=true
    bUseGravityOverride=true
    gravityScale=0.35
    simulationSpace=RootRelative
    alphaBoolBlend {blendOption=Linear}
    boundBone {boneName=dyn_shawl_frontdelt_l_01}
    componentAppliedLinearAccClamp=<1000 1000 1300>
    componentLinearAccScale=<0.5 0.5 1>
    componentLinearVelScale=<0.5 0.5 1>
    gravityOverride {z=-1500}
    physicsBodyDefinitions [{
      boundBone {boneName=dyn_shawl_frontdelt_l_01}
      boxExtents=<20 35 35>
      constraintSetup {angularLimitsMax=<5 5 2>, angularLimitsMin=<-5 -5 -15>, angularTarget=<1 0 0>}
      localJointOffset=<4 0 0>
     }]
   }
   n39 = Anim Dynamics - Bone: dyn_shawl_frontdelt_l_02!(SimulationSpace:AnimPhysSimSpaceType Enum, ComponentPose=@n38, Alpha=0.5, ExternalForce="0,0,0") -> (Pose) {
    angularDampingOverride=0.995
    angularSpringConstant=10000
    bAngularSpring=true
    bOverrideAngularDamping=true
    bUseGravityOverride=true
    gravityScale=0.35
    simulationSpace=RootRelative
    alphaBoolBlend {blendOption=Linear}
    boundBone {boneName=dyn_shawl_frontdelt_l_02}
    componentAppliedLinearAccClamp=<1000 1000 1300>
    componentLinearAccScale=<0.5 0.5 1>
    componentLinearVelScale=<0.5 0.5 1>
    gravityOverride {z=-1500}
    physicsBodyDefinitions [{
      boundBone {boneName=dyn_shawl_frontdelt_l_02}
      boxExtents=<20 35 35>
      constraintSetup {angularLimitsMax=<12.5 12.5 2>, angularLimitsMin=<-12.5 -12.5 -25>, angularTarget=<1 0 0>}
      localJointOffset=<4 0 0>
     }]
   }
   n40 = Anim Dynamics - Bone: dyn_shawl_shoulder_r_01!(SimulationSpace:AnimPhysSimSpaceType Enum, ComponentPose=@n39, Alpha=0.8, ExternalForce="0,0,0") -> (Pose) {
    angularDampingOverride=0.995
    angularSpringConstant=10000
    bAngularSpring=true
    bOverrideAngularDamping=true
    bUseGravityOverride=true
    gravityScale=0.35
    simulationSpace=RootRelative
    alphaBoolBlend {blendOption=Linear}
    boundBone {boneName=dyn_shawl_shoulder_r_01}
    componentAppliedLinearAccClamp=<1000 1000 1300>
    componentLinearAccScale=<0.5 0.5 1>
    componentLinearVelScale=<0.5 0.5 1>
    gravityOverride {z=-1500}
    physicsBodyDefinitions [{
      boundBone {boneName=dyn_shawl_shoulder_r_01}
      boxExtents=<20 35 35>
      constraintSetup {angularLimitsMax=<5 5 2>, angularLimitsMin=<-5 -5 -15>, angularTarget=<1 0 0>}
      localJointOffset=<-4 0 0>
     }]
   }
   n41 = Anim Dynamics - Bone: dyn_shawl_shoulder_r_02!(SimulationSpace:AnimPhysSimSpaceType Enum, ComponentPose=@n40, Alpha=0.8, ExternalForce="0,0,0") -> (Pose) {
    angularDampingOverride=0.995
    angularSpringConstant=10000
    bAngularSpring=true
    bOverrideAngularDamping=true
    bUseGravityOverride=true
    gravityScale=0.35
    simulationSpace=RootRelative
    alphaBoolBlend {blendOption=Linear}
    boundBone {boneName=dyn_shawl_shoulder_r_02}
    componentAppliedLinearAccClamp=<1000 1000 1300>
    componentLinearAccScale=<0.5 0.5 1>
    componentLinearVelScale=<0.5 0.5 1>
    gravityOverride {z=-1500}
    physicsBodyDefinitions [{
      boundBone {boneName=dyn_shawl_shoulder_r_02}
      boxExtents=<20 35 35>
      constraintSetup {angularLimitsMax=<5 5 2>, angularLimitsMin=<-5 -5 -25>, angularTarget=<1 0 0>}
      localJointOffset=<-4 0 0>
     }]
   }
   n42 = Anim Dynamics - Bone: dyn_shawl_shoulder_r_03!(SimulationSpace:AnimPhysSimSpaceType Enum, ComponentPose=@n41, Alpha=0.8, ExternalForce="0,0,0") -> (Pose) {
    angularDampingOverride=0.995
    angularSpringConstant=10000
    bAngularSpring=true
    bOverrideAngularDamping=true
    bUseGravityOverride=true
    gravityScale=0.35
    simulationSpace=RootRelative
    alphaBoolBlend {blendOption=Linear}
    boundBone {boneName=dyn_shawl_shoulder_r_03}
    componentAppliedLinearAccClamp=<1000 1000 1300>
    componentLinearAccScale=<0.5 0.5 1>
    componentLinearVelScale=<0.5 0.5 1>
    gravityOverride {z=-1500}
    physicsBodyDefinitions [{
      boundBone {boneName=dyn_shawl_shoulder_r_03}
      boxExtents=<20 35 35>
      constraintSetup {angularLimitsMax=<12.5 12.5 2>, angularLimitsMin=<-12.5 -12.5 -25>, angularTarget=<1 0 0>}
      localJointOffset=<-4 0 0>
     }]
   }
   n43 = Anim Dynamics - Bone: dyn_shawl_reardelt_r_01!(SimulationSpace:AnimPhysSimSpaceType Enum, ComponentPose=@n42, Alpha=0.8, ExternalForce="0,0,0") -> (Pose) {
    angularDampingOverride=0.995
    angularSpringConstant=9000
    bAngularSpring=true
    bOverrideAngularDamping=true
    bUseGravityOverride=true
    gravityScale=0.35
    simulationSpace=RootRelative
    alphaBoolBlend {blendOption=Linear}
    boundBone {boneName=dyn_shawl_reardelt_r_01}
    componentAppliedLinearAccClamp=<1000 1000 1300>
    componentLinearAccScale=<0.5 0.5 1>
    componentLinearVelScale=<0.5 0.5 1>
    gravityOverride {z=-1500}
    physicsBodyDefinitions [{
      boundBone {boneName=dyn_shawl_reardelt_r_01}
      boxExtents=<20 35 35>
      constraintSetup {angularLimitsMax=<5 5 2>, angularLimitsMin=<-5 -5 -15>, angularTarget=<1 0 0>}
      localJointOffset=<-4 -0.5 0.5>
     }]
   }
   n44 = Anim Dynamics - Bone: dyn_shawl_reardelt_r_02!(SimulationSpace:AnimPhysSimSpaceType Enum, ComponentPose=@n43, Alpha=0.8, ExternalForce="0,0,0") -> (Pose) {
    angularDampingOverride=0.995
    angularSpringConstant=10000
    bAngularSpring=true
    bOverrideAngularDamping=true
    bUseGravityOverride=true
    gravityScale=0.35
    simulationSpace=RootRelative
    alphaBoolBlend {blendOption=Linear}
    boundBone {boneName=dyn_shawl_reardelt_r_02}
    componentAppliedLinearAccClamp=<1000 1000 1300>
    componentLinearAccScale=<0.5 0.5 1>
    componentLinearVelScale=<0.5 0.5 1>
    gravityOverride {z=-1500}
    physicsBodyDefinitions [{
      boundBone {boneName=dyn_shawl_reardelt_r_02}
      boxExtents=<20 35 35>
      constraintSetup {angularLimitsMax=<12.5 12.5 2>, angularLimitsMin=<-12.5 -12.5 -25>, angularTarget=<1 0 0>}
      localJointOffset=<-4 0 0>
     }]
   }
   n45 = Anim Dynamics - Bone: dyn_shawl_scapula_r_01!(SimulationSpace:AnimPhysSimSpaceType Enum, ComponentPose=@n44, Alpha=0.8, ExternalForce="0,0,0") -> (Pose) {
    angularDampingOverride=0.995
    angularSpringConstant=12000
    bAngularSpring=true
    bOverrideAngularDamping=true
    bUseGravityOverride=true
    gravityScale=0.35
    simulationSpace=RootRelative
    alphaBoolBlend {blendOption=Linear}
    boundBone {boneName=dyn_shawl_scapula_r_01}
    componentAppliedLinearAccClamp=<1000 1000 1300>
    componentLinearAccScale=<0.5 0.5 1>
    componentLinearVelScale=<0.5 0.5 1>
    gravityOverride {z=-1500}
    physicsBodyDefinitions [{
      boundBone {boneName=dyn_shawl_scapula_r_01}
      boxExtents=<20 35 35>
      constraintSetup {angularLimitsMax=<5 15 0>, angularLimitsMin=<-5 -15 -30>, angularTarget=<1 0 0>}
      localJointOffset=<-4 0.5 -0.5>
     }]
   }
   n46 = Anim Dynamics - Bone: dyn_shawl_scapula_r_02!(SimulationSpace:AnimPhysSimSpaceType Enum, ComponentPose=@n45, Alpha=0.8, ExternalForce="0,0,0") -> (Pose) {
    angularDampingOverride=0.995
    angularSpringConstant=10000
    bAngularSpring=true
    bOverrideAngularDamping=true
    bUseGravityOverride=true
    gravityScale=0.35
    simulationSpace=RootRelative
    alphaBoolBlend {blendOption=Linear}
    boundBone {boneName=dyn_shawl_scapula_r_02}
    componentAppliedLinearAccClamp=<1000 1000 1300>
    componentLinearAccScale=<0.5 0.5 1>
    componentLinearVelScale=<0.5 0.5 1>
    gravityOverride {z=-1500}
    physicsBodyDefinitions [{
      boundBone {boneName=dyn_shawl_scapula_r_02}
      boxExtents=<20 35 35>
      constraintSetup {angularLimitsMax=<12.5 15 0>, angularLimitsMin=<-12.5 -15 -25>, angularTarget=<1 0 0>}
      localJointOffset=<-4 0 0>
     }]
   }
   n47 = Anim Dynamics - Bone: dyn_shawl_frontdelt_r_01!(SimulationSpace:AnimPhysSimSpaceType Enum, ComponentPose=@n46, Alpha=0.5, ExternalForce="0,0,0") -> (Pose) {
    angularDampingOverride=0.995
    angularSpringConstant=10000
    bAngularSpring=true
    bOverrideAngularDamping=true
    bUseGravityOverride=true
    gravityScale=0.35
    simulationSpace=RootRelative
    alphaBoolBlend {blendOption=Linear}
    boundBone {boneName=dyn_shawl_frontdelt_r_01}
    componentAppliedLinearAccClamp=<1000 1000 1300>
    componentLinearAccScale=<0.5 0.5 1>
    componentLinearVelScale=<0.5 0.5 1>
    gravityOverride {z=-1500}
    physicsBodyDefinitions [{
      boundBone {boneName=dyn_shawl_frontdelt_r_01}
      boxExtents=<20 35 35>
      constraintSetup {angularLimitsMax=<5 5 2>, angularLimitsMin=<-5 -5 -15>, angularTarget=<1 0 0>}
      localJointOffset=<-4 0 0>
     }]
   }
   n48 = Anim Dynamics - Bone: dyn_shawl_frontdelt_r_02!(SimulationSpace:AnimPhysSimSpaceType Enum, ComponentPose=@n47, Alpha=0.5, ExternalForce="0,0,0") -> (Pose) {
    angularDampingOverride=0.995
    angularSpringConstant=10000
    bAngularSpring=true
    bOverrideAngularDamping=true
    bUseGravityOverride=true
    gravityScale=0.35
    simulationSpace=RootRelative
    alphaBoolBlend {blendOption=Linear}
    boundBone {boneName=dyn_shawl_frontdelt_r_02}
    componentAppliedLinearAccClamp=<1000 1000 1300>
    componentLinearAccScale=<0.5 0.5 1>
    componentLinearVelScale=<0.5 0.5 1>
    gravityOverride {z=-1500}
    physicsBodyDefinitions [{
      boundBone {boneName=dyn_shawl_frontdelt_r_02}
      boxExtents=<20 35 35>
      constraintSetup {angularLimitsMax=<12.5 12.5 2>, angularLimitsMin=<-12.5 -12.5 -25>, angularTarget=<1 0 0>}
      localJointOffset=<-4 0 0>
     }]
   }
   Save cached pose 'Post Main Anim Dynamics'(Pose=Component To Local(ComponentPose=@n48)) {cacheName=Post Main Anim Dynamics}
   n49 = Control Rig!(
    Source=Copy Pose From Mesh() {bUseAttachedParent=true}
    Alpha
    Is Host Full:bool=True
    Is Knee Sliding:bool=Get bIsKneeSliding()
    Is Baseball Sliding:bool=Get bIsBaseballSliding()
   ) -> (Pose) {
    alphaBoolBlend {blendOption=Linear}
    controlRigAssetReference {blueprintRigClass=&/M_MED_ClawPad_HostFull_CtrlRig.M_MED_ClawPad_HostFull_CtrlRig_C}
   }
   Save cached pose 'PostDeformationRig'(Pose=@n49) {cacheName=PostDeformationRig}
   n50 = "Transform (Modify) Bone - Bone: dyn_skirt_ground_collision"!(
    ComponentPose=Local To Component(LocalPose=Use cached pose 'PostDeformationRig'())
    bAlphaBoolEnabled:bool=True
    Translation="0,0,0"
    Rotation="0,0,0"
    Scale="1,1,1"
   ) -> (Pose) {
    alphaInputType=Bool
    rotationMode=BMM_Replace
    rotationSpace=BCS_WorldSpace
    translationMode=BMM_Additive
    translationSpace=BCS_BoneSpace
    alphaBoolBlend {blendOption=Linear}
    boneToModify {boneName=dyn_skirt_ground_collision}
   }
   n51 = "Transform (Modify) Bone - Bone: dyn_skirt_ground_collision"(ComponentPose=@n50, bAlphaBoolEnabled=True, Translation="0,0,-4", Rotation="0,0,0", Scale="1,1,1") {
    alphaInputType=Bool
    rotationSpace=BCS_WorldSpace
    translationMode=BMM_Replace
    translationSpace=BCS_BoneSpace
    alphaBoolBlend {blendInTime=0.1, blendOption=Linear, blendOutTime=0.1}
    boneToModify {boneName=dyn_skirt_ground_collision}
    translation {z=-4}
   }
   n52 = "Constraint - dyn_skirt_SIMSPACE (P)"(ComponentPose=@n51, Alpha=1, ConstraintWeights_0=1) {
    alphaBoolBlend {blendOption=Linear}
    alphaScaleBiasClamp {bClampResult=true, bInterpResult=true, interpSpeedDecreasing=2, interpSpeedIncreasing=2}
    boneToModify {boneName=dyn_skirt_SIMSPACE}
    constraintSetup [{offsetOption=None, targetBone {boneName=root}}]
    constraintWeights [1]
   }
   n53 = "Transform (Modify) Bone - Bone: dyn_front_skirt_l_MSTR"!(ComponentPose=@n52, bAlphaBoolEnabled:bool=True, Translation="0,0,0", Rotation="0.418,-20,15.155", Scale="1,1,1") -> (Pose) {
    alphaInputType=Bool
    rotationMode=BMM_Additive
    rotationSpace=BCS_BoneSpace
    translationMode=BMM_Additive
    translationSpace=BCS_BoneSpace
    alphaBoolBlend {blendInTime=0.2, blendOption=QuadraticInOut, blendOutTime=0.2}
    boneToModify {boneName=dyn_front_skirt_l_MSTR}
    rotation=<rot 0.418 -20 15.155>
   }
   n54 = "Transform (Modify) Bone - Bone: dyn_side_skirt_l_MSTR"!(ComponentPose=@n53, bAlphaBoolEnabled:bool=True, Translation="0,0,0", Rotation="0.004,-15,0.001", Scale="1,1,1") -> (Pose) {
    alphaInputType=Bool
    rotationMode=BMM_Additive
    rotationSpace=BCS_BoneSpace
    translationMode=BMM_Additive
    translationSpace=BCS_BoneSpace
    alphaBoolBlend {blendInTime=0.2, blendOption=QuadraticInOut, blendOutTime=0.2}
    boneToModify {boneName=dyn_side_skirt_l_MSTR}
    rotation=<rot 0.004 -15 0.001>
   }
   n55 = "Transform (Modify) Bone - Bone: dyn_skirt_ground_collision"!(ComponentPose=@n54, Alpha:float=1, Translation="0,0,-80", Rotation="0,0,0", Scale="1,1,1") -> (Pose) {
    translationMode=BMM_Additive
    translationSpace=BCS_BoneSpace
    alphaBoolBlend {blendOption=Linear}
    boneToModify {boneName=dyn_skirt_ground_collision}
    translation {z=-80}
   }
   Save cached pose 'Post Slope Posing'(Pose=Component To Local(ComponentPose=@n55)) {cacheName=Post Slope Posing}
  ]}
 EventGraph {nodes [
   n1 = Event BlueprintUpdateAnimation()
   > n2 = Sequence!(execute) -> (then_0, then_1, then_2:Exec, then_3:Exec, then_4:Exec)
   Set RBAN Alpha(execute=@n2.then_3, RBAN Alpha:double=1) -> (Output_Get:double)
   > Branch(Condition=Get bIsDBNOCarried())
   > Set RBAN Alpha(RBAN Alpha=0.75)
   n3 = Branch(execute=@n2.then_1, Condition=Get Delay Slope Calc() -> (Delay Slope Calc:bool))
   Handle Slope Warp!(execute=@n3.else, self, DeltaTime:double=@n1.DeltaTimeX) -> (then)
   n4 = Branch(execute=@n2.then_2, Condition=Get bIsInFrontEnd())
   > Set Simulation Space(Simulation Space:AnimPhysSimSpaceType Enum=World) -> (Output_Get:AnimPhysSimSpaceType Enum)
   > Set AnimDyn Shawl Alpha(AnimDyn Shawl Alpha:double=0.25) -> (Output_Get:double)
   > Set AnimDyn Panel Alpha(AnimDyn Panel Alpha:double=0.35) -> (Output_Get:double)
   > Set Custom Idle Is Playing Float(Custom Idle Is Playing Float:float=GetCurveValue(self=GetOwnerFrontEndAnimInstance(), CurveName=Frontend_Idle_Bypass)) -> (Output_Get:float)
   > Set Sleeve Sim Space(Sleeve Sim Space:Sim Space Settings Structure=Get Sleeve Sim Space Frontend() -> (Sleeve Sim Space Frontend:Sim Space Settings Structure)) -> (Output_Get:Sim Space Settings Structure)
   Set Simulation Space(execute=@n4.else)
   > Set AnimDyn Shawl Alpha(AnimDyn Shawl Alpha=0.8)
   > Set AnimDyn Panel Alpha(AnimDyn Panel Alpha=1)
   > Set Sleeve Sim Space(Sleeve Sim Space=Get Sleeve Sim Space Normal() -> (Sleeve Sim Space Normal:Sim Space Settings Structure))
   n5 = NOT Boolean(A=OR Boolean(A=Get bIsDBNOCarried(), B=Get bIsDBNOCarrying() -> (bIsDBNOCarrying:bool)))
   n6 = Get bIsDBNO()
   n7 = Branch(execute=@n2.then_4, Condition=OR Boolean(A=Get bIsTacticalSprinting(), B=AND Boolean(A=@n6, B=@n5)))
   > Set SetSimSpaceParent(SetSimSpaceParent:bool=true) -> (Output_Get:bool)
   Set SetSimSpaceParent(execute=@n7.else, SetSimSpaceParent=false)
   n8 = Gravity Override State Function DEPRECATED(
    execute=@n2.then_0
    CharacterPartAnimInstance=Self-Reference()
    GameTimeInSeconds=GetGameTimeInSeconds()
    GravityOverrideParams=Get Gravity Override Params()
    RBAN Params=Get RBAN Params()
   )
   > Set Gravity Override(Gravity Override=@n8)
   > Set RBAN Settings Out(RBAN Settings Out:RBANSetting Struct Structure=@n8.RBAN SettingsOut) -> (Output_Get:RBANSetting Struct Structure)
   > n9 = Sequence!(execute) -> (then_0, then_1)
   > Branch(Condition=Get bIsSliding())
   > Set RBAN Settings Out(RBAN Settings Out=Get RBAN Settings Sliding() -> (RBAN Settings Sliding:RBANSetting Struct Structure))
   Branch(execute=@n9.then_1, Condition=AND Boolean(A=Get bIsInFrontEnd(), B=Get bIsPlayingEmote()))
   > Set RBAN Settings Out()
   > Set Gravity Override(Gravity Override="0,0,-980")
   Event OnCharacterCustomizationLoaded()
   > Set IsClawpadGlider(IsClawpadGlider:bool=GliderItemDefinitionHasAnyMetaTag!(Tags:Array of Gameplay Tag Structures=Get Tags() -> (Tags:Array of Gameplay Tag Structures), bRequiresOpened:bool=false) -> (ReturnValue:bool)) -> (Output_Get:bool)
   > Delay(Duration=2)
   > Set Delay Slope Calc(Delay Slope Calc:bool=false) -> (Output_Get:bool)
   Event OnParachuteOpened()
   > Set RBAN Alpha(RBAN Alpha=0.75)
   > Branch(Condition=Get IsClawpadGlider() -> (IsClawpadGlider:bool))
   > Set IsGlidingOnCustomGlider(IsGlidingOnCustomGlider:bool=true) -> (Output_Get:bool)
   Event OnParachuteClosed()
   > Set RBAN Alpha(RBAN Alpha=1)
   > Set IsGlidingOnCustomGlider(IsGlidingOnCustomGlider=false)
  ]}
 Handle Slope Warp {nodes [
   n1 = Handle Slope Warp!() -> (then, DeltaTime:double=~) {is=title K2Node_FunctionEntry}
   > n2 = Sequence()
   n3 = Get bIsOnGround() -> (bIsOnGround:bool)
   > n4 = Branch(Condition=OR Boolean!(A=NOT Boolean(A=@n3), B=Get bIsPlayingEmote(), C:bool=Get bIsInVehicle() -> (bIsInVehicle:bool)) -> (ReturnValue))
   > Set Target Info Input Root Slope Translation(Target Info Input Root Slope Translation:Vector="0,0,-100") -> (Output_Get:Vector)
   Set Slope Info Input Slope Rotation(
    execute=@n4.else
    Slope Info Input Slope Rotation:Rotator=RInterpTo!(
     Current:Rotator=Get Slope Info Input Slope Rotation() -> (Slope Info Input Slope Rotation:Rotator)
     Target:Rotator=Get!() -> (
      SlopeInfoInput_RootSlopeTranslation:Vector=~
      SlopeInfoInput_LeftFootSlopeTranslation:Vector=~
      SlopeInfoInput_RightFootSlopeTranslation:Vector=~
      SlopeInfoInput_SlopeRotation:Rotator=~
     ).SlopeInfoInput_SlopeRotation
     DeltaTime:float=@n1
     InterpSpeed:float=7
    ) -> (ReturnValue:Rotator)
   ) -> (Output_Get:Rotator)
   > Set Target Info Input Root Slope Translation(Target Info Input Root Slope Translation="0,0,5")
   Set Slope Info Input Root Slope Translation(
    execute=@n2.then_1
    Slope Info Input Root Slope Translation:Vector=VInterpTo!(
     Current:Vector=Get Slope Info Input Root Slope Translation() -> (Slope Info Input Root Slope Translation:Vector)
     Target:Vector=Get Target Info Input Root Slope Translation() -> (Target Info Input Root Slope Translation:Vector)
     DeltaTime:float=@n1
     InterpSpeed:float=10
    ) -> (ReturnValue:Vector)
   ) -> (Output_Get:Vector)
  ]}
]
variables [
 AnimDyn Panel Alpha {type=number}
 AnimDyn Shawl Alpha {type=number}
 Cloth Core Settings {
  type=ClothCoreSettingStruct
  default {
   collisionThickness=0
   dampingCoefficient=0.05
   frictionCoefficient=0.05
   localDampingCoefficient=0
   numberIterations=2
   numberSubsteps=1
   selfCollisionThickness=0
   animDriveDamping=<0 1>
   animDriveStiffness=<0 1>
   areaStiffness=<1 1>
   bendingStiffness=<1 1>
   edgeStiffness=<1 1>
   tetherStiffness=<1 1>
  }
 }
 Cloth Core Settings_0 {
  type=ClothCoreSettingStruct
  default {
   collisionThickness=5
   dampingCoefficient=0.3
   frictionCoefficient=0.8
   localDampingCoefficient=0
   numberIterations=2
   numberSubsteps=1
   selfCollisionThickness=0
   animDriveDamping=<0.2 1>
   animDriveStiffness=<0.2 1>
   areaStiffness=<1 1>
   bendingStiffness=<1 1>
   edgeStiffness=<1 1>
   tetherStiffness=<1 1>
  }
 }
 Cloth Paramaters {
  type=ClothParamsStruct
  default {
   dBNO {
    angularVelocityScale=0.6
    fictitiousAngularScale=0.1
    jointName=pelvis
    dragCoefficient=<0.2 1>
    gravityOverride=<-980 400 0>
    liftCoefficient=<0 0.5>
    linearVelocityScale=<0.3 0.3 0.1>
   }
   dBNO_Carried {
    angularVelocityScale=0.6
    fictitiousAngularScale=0.1
    jointName=pelvis
    dragCoefficient=<0.2 1>
    gravityOverride=<-980 -200 0>
    liftCoefficient=<0 0.5>
    linearVelocityScale=<0.3 0.3 0.1>
   }
   dBNO_Impulse {
    angularVelocityScale=0
    fictitiousAngularScale=0
    jointName=pelvis
    dragCoefficient=<0 0.2>
    gravityOverride=<-3000 -2000 0>
    liftCoefficient=<0 0.1>
    linearVelocityScale=<0.3 0.1 0.3>
   }
   dBNO_RollBwd {
    angularVelocityScale=0
    fictitiousAngularScale=0
    jointName=pelvis
    dragCoefficient=<0 0.2>
    gravityOverride=<-2000 0 0>
    liftCoefficient=<0 0.1>
    linearVelocityScale=<0.3 0.1 0.3>
   }
   dBNO_RollFwd {
    angularVelocityScale=0
    fictitiousAngularScale=0
    jointName=pelvis
    dragCoefficient=<0 0.2>
    gravityOverride=<-980 0 0>
    liftCoefficient=<0 0.1>
    linearVelocityScale=<0.3 0.1 0.3>
   }
   dBNO_Tumbling {
    angularVelocityScale=0
    fictitiousAngularScale=0
    jointName=pelvis
    dragCoefficient=<0 0.2>
    gravityOverride=<-2500 -300 0>
    liftCoefficient=<0 0.1>
    linearVelocityScale=<0.3 0.1 0.3>
   }
   emoteOrMelee {
    angularVelocityScale=0.6
    fictitiousAngularScale=0.1
    jointName=root
    dragCoefficient=<2.5 2.5>
    gravityOverride=<0 0 -980>
    liftCoefficient=<0.5 0.5>
    linearVelocityScale=<0.6 0.6 0.5>
   }
   front_End {
    angularVelocityScale=0.05
    fictitiousAngularScale=0.6
    jointName=root
    dragCoefficient=<0.2 0.2>
    gravityOverride=<0 -200 -980>
    liftCoefficient=<0.1 0.1>
    linearVelocityScale=<0.1 0.1 0.1>
   }
   onGround_Crouch_Moving {
    angularVelocityScale=0.1
    fictitiousAngularScale=0.3
    jointName=root
    dragCoefficient=<2 2>
    gravityOverride=<0 0 -980>
    liftCoefficient=<0.5 0.5>
    linearVelocityScale=<0.6 0.6 0.5>
   }
   onGround_Falling {
    angularVelocityScale=0.6
    fictitiousAngularScale=0
    jointName=root
    dragCoefficient=<0 0.2>
    gravityOverride=<0 -500 -980>
    liftCoefficient=<0 0.1>
    linearVelocityScale=<2 1 0.3>
   }
   onGround_Moving {
    angularVelocityScale=0.6
    fictitiousAngularScale=0.1
    jointName=root
    dragCoefficient=<2 2>
    gravityOverride=<0 0 -980>
    liftCoefficient=<0.5 0.5>
    linearVelocityScale=<0.3 0.3 0.2>
   }
   onGround_Standing {
    angularVelocityScale=0.05
    fictitiousAngularScale=1.2
    jointName=root
    dragCoefficient=<2.5 2.5>
    gravityOverride=<0 0 -980>
    liftCoefficient=<0.5 0.5>
    linearVelocityScale=<0.6 0.6 0.5>
   }
   rail_Grinding {
    angularVelocityScale=0.2
    fictitiousAngularScale=0
    jointName=root
    dragCoefficient=<0.1 0.2>
    gravityOverride=<50 -500 -400>
    liftCoefficient=<0 0.5>
    linearVelocityScale=<0.1 0.1 0.1>
   }
   riding_Moving {
    angularVelocityScale=0.05
    fictitiousAngularScale=1
    jointName=root
    dragCoefficient=<0.5 0.5>
    gravityOverride=<0 0 -980>
    liftCoefficient=<0.5 0.5>
    linearVelocityScale=<0.4 0.6 0.3>
   }
   riding_Moving_Targeting {
    angularVelocityScale=0.05
    fictitiousAngularScale=0.7
    jointName=root
    dragCoefficient=<0.5 0.5>
    gravityOverride=<0 0 -980>
    liftCoefficient=<0.5 0.5>
    linearVelocityScale=<0.2 0.6 0.2>
   }
   riding_Standing {
    angularVelocityScale=0.05
    fictitiousAngularScale=1
    jointName=root
    dragCoefficient=<0.5 0.5>
    gravityOverride=<0 -450 -980>
    liftCoefficient=<0.5 0.5>
    linearVelocityScale=<0.6 0.6 0.3>
   }
   skydive_DivingDown {
    angularVelocityScale=0.1
    fictitiousAngularScale=0
    jointName=pelvis
    dragCoefficient=<0.3 0.3>
    gravityOverride=<-450 -100 0>
    liftCoefficient=<0.9 0.9>
    linearVelocityScale=<0.1 0.1 0.1>
   }
   skydive_FloatingDown {
    angularVelocityScale=0.1
    fictitiousAngularScale=0
    jointName=root
    dragCoefficient=<0.3 0.3>
    gravityOverride=<0 -800 200>
    liftCoefficient=<1.3 1.3>
    linearVelocityScale=<0.2 0.2 0.1>
   }
   skydive_GoingUp {
    angularVelocityScale=0.1
    fictitiousAngularScale=0
    jointName=root
    dragCoefficient=<0.01 0.01>
    gravityOverride=<0 -2000 400>
    liftCoefficient=<0.05 0.05>
    linearVelocityScale=<0.1 0.1 0.01>
   }
   skydive_Parachuting {
    angularVelocityScale=0.6
    fictitiousAngularScale=0.1
    jointName=root
    dragCoefficient=<0.8 0.8>
    gravityOverride=<0 -600 -500>
    liftCoefficient=<0.8 0.8>
    linearVelocityScale=<0.2 0.2 0.3>
   }
   skydive_Surfing {
    angularVelocityScale=0.6
    fictitiousAngularScale=0.1
    jointName=root
    dragCoefficient=<1.5 1.5>
    gravityOverride=<0 -600 -500>
    liftCoefficient=<1 1>
    linearVelocityScale=<0.1 0.1 0.3>
   }
   swimming {
    angularVelocityScale=0.6
    fictitiousAngularScale=0
    jointName=root
    dragCoefficient=<2.5 2.5>
    gravityOverride=<0 0 0>
    liftCoefficient=<0.1 0.1>
    linearVelocityScale=<0.6 0.6 0.5>
   }
   tactical_Sprint {
    angularVelocityScale=0.6
    fictitiousAngularScale=0.1
    jointName=root
    dragCoefficient=<2 2>
    gravityOverride=<0 0 -980>
    liftCoefficient=<0.5 0.5>
    linearVelocityScale=<0.1 0.1 0.1>
   }
   ziplining {
    angularVelocityScale=0.6
    fictitiousAngularScale=0
    jointName=root
    dragCoefficient=<0 1.5>
    gravityOverride=<0 0 -980>
    liftCoefficient=<0 0.3>
    linearVelocityScale=<0.4 0.4 0.5>
   }
  }
 }
 Cloth Paramaters_0 {
  type=ClothParamsStruct
  default {
   dBNO {
    angularVelocityScale=0.6
    fictitiousAngularScale=0.1
    jointName=pelvis
    dragCoefficient=<0.2 1>
    gravityOverride=<-980 400 0>
    liftCoefficient=<0 0.5>
    linearVelocityScale=<0.3 0.3 0.1>
   }
   dBNO_Carried {
    angularVelocityScale=0.6
    fictitiousAngularScale=0.1
    jointName=pelvis
    dragCoefficient=<0.2 1>
    gravityOverride=<-980 -200 0>
    liftCoefficient=<0 0.5>
    linearVelocityScale=<0.3 0.3 0.1>
   }
   dBNO_Impulse {
    angularVelocityScale=0
    fictitiousAngularScale=0
    jointName=pelvis
    dragCoefficient=<0 0.2>
    gravityOverride=<-3000 -2000 0>
    liftCoefficient=<0 0.1>
    linearVelocityScale=<0.3 0.1 0.3>
   }
   dBNO_RollBwd {
    angularVelocityScale=0
    fictitiousAngularScale=0
    jointName=pelvis
    dragCoefficient=<0 0.2>
    gravityOverride=<-2000 0 0>
    liftCoefficient=<0 0.1>
    linearVelocityScale=<0.3 0.1 0.3>
   }
   dBNO_RollFwd {
    angularVelocityScale=0
    fictitiousAngularScale=0
    jointName=pelvis
    dragCoefficient=<0 0.2>
    gravityOverride=<-980 0 0>
    liftCoefficient=<0 0.1>
    linearVelocityScale=<0.3 0.1 0.3>
   }
   dBNO_Tumbling {
    angularVelocityScale=0
    fictitiousAngularScale=0
    jointName=pelvis
    dragCoefficient=<0 0.2>
    gravityOverride=<-2500 -300 0>
    liftCoefficient=<0 0.1>
    linearVelocityScale=<0.3 0.1 0.3>
   }
   emoteOrMelee {
    angularVelocityScale=0.2
    fictitiousAngularScale=0.1
    jointName=pelvis
    dragCoefficient=<0 0.3>
    gravityOverride=<-980 -500 0>
    liftCoefficient=<0 0.2>
    linearVelocityScale=<0.05 0.05 0.05>
   }
   front_End {
    angularVelocityScale=0.1
    fictitiousAngularScale=0.2
    jointName=root
    dragCoefficient=<0.2 0.2>
    gravityOverride=<0 -200 -980>
    liftCoefficient=<0.1 0.1>
    linearVelocityScale=<0.1 0.1 0.1>
   }
   onGround_Crouch_Moving {
    angularVelocityScale=0.1
    fictitiousAngularScale=0.3
    jointName=root
    dragCoefficient=<2 2>
    gravityOverride=<0 0 -980>
    liftCoefficient=<0.6 0.6>
    linearVelocityScale=<0.6 0.6 0.5>
   }
   onGround_Falling {
    angularVelocityScale=0.6
    fictitiousAngularScale=0
    jointName=root
    dragCoefficient=<0 0.2>
    gravityOverride=<0 -500 -980>
    liftCoefficient=<0 0.1>
    linearVelocityScale=<2 1 0.3>
   }
   onGround_Moving {
    angularVelocityScale=0.3
    fictitiousAngularScale=0.1
    jointName=root
    dragCoefficient=<0 0.3>
    gravityOverride=<0 -100 -980>
    liftCoefficient=<0 0.2>
    linearVelocityScale=<0.3 0.5 0.5>
   }
   onGround_Standing {
    angularVelocityScale=0.05
    fictitiousAngularScale=0.6
    jointName=root
    dragCoefficient=<0 2>
    gravityOverride=<0 -100 -980>
    liftCoefficient=<0 0.5>
    linearVelocityScale=<0.6 1 1>
   }
   rail_Grinding {
    angularVelocityScale=0.2
    fictitiousAngularScale=0
    jointName=root
    dragCoefficient=<0.1 0.2>
    gravityOverride=<50 -500 -400>
    liftCoefficient=<0 0.5>
    linearVelocityScale=<0.1 0.1 0.1>
   }
   riding_Moving {
    angularVelocityScale=0.05
    fictitiousAngularScale=1
    jointName=root
    dragCoefficient=<0.5 0.5>
    gravityOverride=<0 0 -980>
    liftCoefficient=<0.5 0.5>
    linearVelocityScale=<0.4 0.6 0.3>
   }
   riding_Moving_Targeting {
    angularVelocityScale=0.05
    fictitiousAngularScale=0.7
    jointName=root
    dragCoefficient=<0.5 0.5>
    gravityOverride=<0 0 -980>
    liftCoefficient=<0.5 0.5>
    linearVelocityScale=<0.2 0.6 0.2>
   }
   riding_Standing {
    angularVelocityScale=0.05
    fictitiousAngularScale=1
    jointName=root
    dragCoefficient=<0.5 0.5>
    gravityOverride=<0 -450 -980>
    liftCoefficient=<0.5 0.5>
    linearVelocityScale=<0.6 0.6 0.3>
   }
   skydive_DivingDown {
    angularVelocityScale=0.1
    fictitiousAngularScale=0
    jointName=pelvis
    dragCoefficient=<0 1.5>
    gravityOverride=<-450 -100 0>
    liftCoefficient=<0 0.3>
    linearVelocityScale=<0.1 0.1 0.1>
   }
   skydive_FloatingDown {
    angularVelocityScale=0.1
    fictitiousAngularScale=0
    jointName=root
    dragCoefficient=<0 2.5>
    gravityOverride=<0 0 980>
    liftCoefficient=<0 0.4>
    linearVelocityScale=<0.2 0.2 0.1>
   }
   skydive_GoingUp {
    angularVelocityScale=0.1
    fictitiousAngularScale=0
    jointName=root
    dragCoefficient=<0 0>
    gravityOverride=<0 -3000 3000>
    liftCoefficient=<0 0>
    linearVelocityScale=<0.1 0.1 0.05>
   }
   skydive_Parachuting {
    angularVelocityScale=0.1
    fictitiousAngularScale=0
    jointName=root
    dragCoefficient=<0 2.5>
    gravityOverride=<0 0 -980>
    liftCoefficient=<0 0.5>
    linearVelocityScale=<0.2 0.2 0.2>
   }
   skydive_Surfing {
    angularVelocityScale=0.6
    fictitiousAngularScale=0.1
    jointName=root
    dragCoefficient=<1.5 1.5>
    gravityOverride=<0 -600 -500>
    liftCoefficient=<1 1>
    linearVelocityScale=<0.1 0.1 0.3>
   }
   swimming {
    angularVelocityScale=0.6
    fictitiousAngularScale=0
    jointName=root
    dragCoefficient=<0 2>
    gravityOverride=<0 -600 400>
    liftCoefficient=<0 0.4>
    linearVelocityScale=<0.6 0.2 0.2>
   }
   tactical_Sprint {
    angularVelocityScale=0.6
    fictitiousAngularScale=0.1
    jointName=root
    dragCoefficient=<2 2>
    gravityOverride=<0 0 -980>
    liftCoefficient=<0.5 0.5>
    linearVelocityScale=<0.6 0.6 0.5>
   }
   ziplining {
    angularVelocityScale=0.6
    fictitiousAngularScale=0
    jointName=root
    dragCoefficient=<0 1.5>
    gravityOverride=<0 0 -980>
    liftCoefficient=<0 0.3>
    linearVelocityScale=<0.4 0.4 0.5>
   }
  }
 }
 Custom Idle Is Playing Float {type=number}
 Delay Slope Calc {type=boolean, default=true}
 Disable Floor Collision {type=boolean}
 Gravity Override {type=Vector}
 Gravity Override Params {
  type=GravityOverrideParamsStruct
  default {
   onGround_Moving {gravityOverride=<0 -400 -700>}
   skydive_DivingDown {windAmplitude=<50 600 400>, windFrequency=<3 24 20>}
   skydive_FloatingDown {gravityOverride=<0 -700 400>, windAmplitude=<100 -100 150>, windFrequency=<5 12 16>}
   skydive_GoingUp {gravityOverride=<0 -300 450>}
   skydive_Parachuting {windAmplitude=<100 -300 200>}
   skydive_Surfing {jointName=pelvis, gravityOverride=<-1500 0 0>, windAmplitude=<0 0 0>, windFrequency=<0 0 0>}
   tactical_Sprint {gravityOverride=<0 -600 -500>}
  }
 }
 IsClawpadGlider {type=boolean}
 IsGlidingOnCustomGlider {type=boolean}
 RBAN Alpha {type=number, default=1}
 RBAN Params {
  type=RBANParamsStruct
  default {
   dBNO {simSpaceSettings {
     dampingAlpha=0
     maxAngularAcceleration=30
     maxAngularVelocity=7.5
     maxLinearAcceleration=3000
     maxLinearVelocity=400
     velocityScaleZ=0.4
     externalLinearDragV=<0.5 0.5 0.2>
    }}
   dBNO_Carried {simSpaceSettings {
     dampingAlpha=0
     maxAngularAcceleration=30
     maxAngularVelocity=7.5
     maxLinearAcceleration=3000
     maxLinearVelocity=400
     velocityScaleZ=0.4
     externalLinearDragV=<0.5 0.5 0.2>
    }}
   emoteOrMelee {simSpaceSettings {
     dampingAlpha=0
     maxAngularAcceleration=15
     maxAngularVelocity=4
     maxLinearAcceleration=3000
     maxLinearVelocity=400
     velocityScaleZ=0.2
     worldAlpha=0.5
     externalLinearDragV=<0.5 0.5 0.2>
    }}
   front_End {simSpaceSettings {
     dampingAlpha=0
     maxAngularAcceleration=15
     maxAngularVelocity=1.5
     maxLinearAcceleration=3000
     maxLinearVelocity=400
     velocityScaleZ=0.4
     worldAlpha=0.7
     externalLinearDragV=<0 0 0>
    }}
   onGround_Crouch_Moving {simSpaceSettings {
     dampingAlpha=0
     maxAngularAcceleration=30
     maxAngularVelocity=7.5
     maxLinearAcceleration=3000
     maxLinearVelocity=400
     velocityScaleZ=0.4
     externalLinearDragV=<0.5 0.5 0.2>
    }}
   onGround_Falling {simSpaceSettings {
     dampingAlpha=0
     maxAngularAcceleration=30
     maxAngularVelocity=7.5
     maxLinearAcceleration=3000
     maxLinearVelocity=400
     velocityScaleZ=0.4
     externalLinearDragV=<0.5 0.5 0.2>
    }}
   onGround_Moving {
    componentLinearAccScale=<0 0 0>
    componentLinearVelScale=<0 0 0>
    simSpaceSettings {
     dampingAlpha=0
     maxAngularAcceleration=30
     maxAngularVelocity=7.5
     maxLinearAcceleration=3000
     maxLinearVelocity=500
     velocityScaleZ=0.3
     worldAlpha=0.8
     externalLinearDragV=<0.5 0.5 0.2>
    }
   }
   onGround_Standing {
    componentLinearAccScale=<0 0 0>
    componentLinearVelScale=<0 0 0>
    simSpaceSettings {
     dampingAlpha=0
     maxAngularAcceleration=30
     maxAngularVelocity=7.5
     maxLinearAcceleration=3000
     maxLinearVelocity=400
     velocityScaleZ=0.4
     externalLinearDragV=<0.5 0.5 0.2>
    }
   }
   rail_Grinding {simSpaceSettings {
     dampingAlpha=0
     maxAngularAcceleration=30
     maxAngularVelocity=7.5
     maxLinearAcceleration=2000
     maxLinearVelocity=200
     velocityScaleZ=0.2
     worldAlpha=0.5
     externalLinearDragV=<0 0 0>
    }}
   riding_Moving {simSpaceSettings {
     dampingAlpha=0
     maxAngularAcceleration=30
     maxAngularVelocity=7.5
     maxLinearAcceleration=3000
     maxLinearVelocity=400
     velocityScaleZ=0.4
     externalLinearDragV=<0.5 0.5 0.2>
    }}
   riding_Moving_Targeting {simSpaceSettings {
     dampingAlpha=0
     maxAngularAcceleration=30
     maxAngularVelocity=7.5
     maxLinearAcceleration=3000
     maxLinearVelocity=400
     velocityScaleZ=0.4
     externalLinearDragV=<0.5 0.5 0.2>
    }}
   riding_Standing {simSpaceSettings {
     dampingAlpha=0
     maxAngularAcceleration=30
     maxAngularVelocity=7.5
     maxLinearAcceleration=3000
     maxLinearVelocity=400
     velocityScaleZ=0.4
     externalLinearDragV=<0.5 0.5 0.2>
    }}
   skydive_DivingDown {
    componentLinearAccScale=<0 0 0>
    componentLinearVelScale=<0 0 0>
    simSpaceSettings {
     dampingAlpha=0
     maxAngularAcceleration=30
     maxAngularVelocity=7.5
     maxLinearAcceleration=3000
     maxLinearVelocity=400
     velocityScaleZ=0.4
     worldAlpha=0.5
     externalLinearDragV=<0.5 0.5 0.2>
    }
   }
   skydive_FloatingDown {
    componentLinearAccScale=<0 0 0>
    componentLinearVelScale=<0 0 0>
    simSpaceSettings {
     dampingAlpha=0
     maxAngularAcceleration=30
     maxAngularVelocity=7.5
     maxLinearAcceleration=3000
     maxLinearVelocity=400
     velocityScaleZ=0.4
     worldAlpha=0.5
     externalLinearDragV=<0.5 0.5 0.2>
    }
   }
   skydive_GoingUp {simSpaceSettings {
     dampingAlpha=0
     maxAngularAcceleration=30
     maxAngularVelocity=7.5
     maxLinearAcceleration=3000
     maxLinearVelocity=400
     velocityScaleZ=0.1
     worldAlpha=0.5
     externalLinearDragV=<0.5 0.5 0.2>
    }}
   skydive_Parachuting {
    componentLinearAccScale=<0 0 0>
    componentLinearVelScale=<0 0 0>
    simSpaceSettings {
     dampingAlpha=0
     maxAngularAcceleration=30
     maxAngularVelocity=7.5
     maxLinearAcceleration=3000
     maxLinearVelocity=400
     velocityScaleZ=0.4
     worldAlpha=0.5
     externalLinearDragV=<0.5 0.5 0.2>
    }
   }
   skydive_Surfing {simSpaceSettings {
     dampingAlpha=0
     maxAngularAcceleration=15
     maxAngularVelocity=2
     maxLinearAcceleration=750
     maxLinearVelocity=50
     velocityScaleZ=0.2
     worldAlpha=0.2
     externalLinearDragV=<0 0 0>
    }}
   swimming {simSpaceSettings {
     dampingAlpha=0
     maxAngularAcceleration=30
     maxAngularVelocity=7.5
     maxLinearAcceleration=3000
     maxLinearVelocity=400
     velocityScaleZ=0.4
     externalLinearDragV=<0.5 0.5 0.2>
    }}
   tactical_Sprint {
    componentLinearAccScale=<0 0 0>
    componentLinearVelScale=<0 0 0>
    simSpaceSettings {
     dampingAlpha=0
     maxAngularAcceleration=8
     maxAngularVelocity=2
     maxLinearAcceleration=1000
     maxLinearVelocity=150
     velocityScaleZ=0.4
     externalLinearDragV=<0 0.5 0>
     externalLinearVelocity=<0 -200 0>
    }
   }
   ziplining {simSpaceSettings {
     dampingAlpha=0
     maxAngularAcceleration=30
     maxAngularVelocity=7.5
     maxLinearAcceleration=3000
     maxLinearVelocity=400
     velocityScaleZ=0.4
     externalLinearDragV=<0.5 0.5 0.2>
    }}
  }
 }
 RBAN Settings Out {
  type=RBANSettingStruct
  default {
   componentLinearAccScale=<0 0 0>
   componentLinearVelScale=<0 0 0>
   simSpaceSettings {
    dampingAlpha=1
    maxAngularAcceleration=1000
    maxAngularVelocity=3
    maxLinearAcceleration=300
    maxLinearVelocity=700
    velocityScaleZ=0
    worldAlpha=1
    externalAngularVelocity=<0 0 0>
    externalLinearDragV=<0 0.5 0>
    externalLinearVelocity=<0 0 0>
   }
  }
 }
 RBAN Settings Sliding {
  type=RBANSettingStruct
  default {
   componentLinearAccScale=<0 0 0>
   componentLinearVelScale=<0 0 0>
   simSpaceSettings {
    dampingAlpha=0
    maxAngularAcceleration=10
    maxAngularVelocity=4
    maxLinearAcceleration=1000
    maxLinearVelocity=200
    velocityScaleZ=0.4
    worldAlpha=0.4
    externalAngularVelocity=<0 0 0>
    externalLinearDragV=<0 0 0>
    externalLinearVelocity=<0 0 0>
   }
  }
 }
 SetSimSpaceParent {type=boolean}
 Simulation Space {type=AnimPhysSimSpaceType, default=RootRelative}
 Sleeve Sim Space {
  type=SimSpaceSettings
  default {
   dampingAlpha=0
   maxAngularAcceleration=30
   maxAngularVelocity=5
   maxLinearAcceleration=1000
   maxLinearVelocity=100
   velocityScaleZ=0.2
   worldAlpha=0.8
   externalLinearDragV=<0.5 0.5 0.5>
  }
 }
 Sleeve Sim Space Frontend {
  type=SimSpaceSettings
  default {
   dampingAlpha=0
   maxAngularAcceleration=40
   maxAngularVelocity=6
   maxLinearAcceleration=1000
   maxLinearVelocity=100
   velocityScaleZ=0.2
   worldAlpha=0.7
   externalLinearDragV=<0 0 0>
  }
 }
 Sleeve Sim Space Normal {
  type=SimSpaceSettings
  default {
   dampingAlpha=0
   maxAngularAcceleration=30
   maxAngularVelocity=5
   maxLinearAcceleration=1000
   maxLinearVelocity=100
   velocityScaleZ=0.2
   worldAlpha=0.8
   externalLinearDragV=<0.5 0.5 0.5>
  }
 }
 Slope Info Input Root Slope Translation {type=Vector}
 Slope Info Input Slope Rotation {type=Rotator}
 Tags {type="array<GameplayTag>", default [{tagName=Cosmetics.MetaTags.ClawPadHost}]}
 Target Info Input Root Slope Translation {type=Vector}
 UseAnimDynamicSleeves {type=boolean}
 Wind Gust Local Settings {type=WindGustLocalSettingStruct, default {elapsedTime=0, finalGustStrengthLocal=0, gustDuration=0, gustStrength=0, timeBetweenGusts=0}}
 Wind Gusts Params {
  type=WindGustParamsStruct
  default {
   dBNO {
    maxDownInterpSpeed=5
    maxGustDuration=2
    maxGustStrength=7
    maxTimeBetweenGusts=7
    maxUpInterpSpeed=5
    minDownInterpSpeed=2
    minGustDuration=0.5
    minGustStrength=3
    minTimeBetweenGusts=0
    minUpInterpSpeed=2
    windTurbulenceScalarA=0.1
    windTurbulenceScalarB=1
    windAmplitude=<0 0 0>
    windFrequency=<0 0 0>
    windOffset=<0 0 0>
   }
   dBNO_Carried {
    maxDownInterpSpeed=5
    maxGustDuration=2
    maxGustStrength=7
    maxTimeBetweenGusts=7
    maxUpInterpSpeed=5
    minDownInterpSpeed=2
    minGustDuration=0.5
    minGustStrength=3
    minTimeBetweenGusts=0
    minUpInterpSpeed=2
    windTurbulenceScalarA=0.1
    windTurbulenceScalarB=1
    windAmplitude=<0 0 0>
    windFrequency=<0 0 0>
    windOffset=<0 0 0>
   }
   dBNO_Impulse {
    maxDownInterpSpeed=5
    maxGustDuration=2
    maxGustStrength=7
    maxTimeBetweenGusts=7
    maxUpInterpSpeed=5
    minDownInterpSpeed=2
    minGustDuration=0.5
    minGustStrength=3
    minTimeBetweenGusts=0
    minUpInterpSpeed=2
    windTurbulenceScalarA=0.1
    windTurbulenceScalarB=1
    windAmplitude=<0 0 0>
    windFrequency=<0 0 0>
    windOffset=<0 0 0>
   }
   dBNO_RollBwd {
    maxDownInterpSpeed=5
    maxGustDuration=2
    maxGustStrength=7
    maxTimeBetweenGusts=7
    maxUpInterpSpeed=5
    minDownInterpSpeed=2
    minGustDuration=0.5
    minGustStrength=3
    minTimeBetweenGusts=0
    minUpInterpSpeed=2
    windTurbulenceScalarA=0.1
    windTurbulenceScalarB=1
    windAmplitude=<0 0 0>
    windFrequency=<0 0 0>
    windOffset=<0 0 0>
   }
   dBNO_RollFwd {
    maxDownInterpSpeed=5
    maxGustDuration=2
    maxGustStrength=7
    maxTimeBetweenGusts=7
    maxUpInterpSpeed=5
    minDownInterpSpeed=2
    minGustDuration=0.5
    minGustStrength=3
    minTimeBetweenGusts=0
    minUpInterpSpeed=2
    windTurbulenceScalarA=0.1
    windTurbulenceScalarB=1
    windAmplitude=<0 0 0>
    windFrequency=<0 0 0>
    windOffset=<0 0 0>
   }
   dBNO_Tumbling {
    maxDownInterpSpeed=5
    maxGustDuration=2
    maxGustStrength=7
    maxTimeBetweenGusts=7
    maxUpInterpSpeed=5
    minDownInterpSpeed=2
    minGustDuration=0.5
    minGustStrength=3
    minTimeBetweenGusts=0
    minUpInterpSpeed=2
    windTurbulenceScalarA=0.1
    windTurbulenceScalarB=1
    windAmplitude=<0 0 0>
    windFrequency=<0 0 0>
    windOffset=<0 0 0>
   }
   emoteOrMelee {
    maxDownInterpSpeed=5
    maxGustDuration=2
    maxGustStrength=7
    maxTimeBetweenGusts=7
    maxUpInterpSpeed=5
    minDownInterpSpeed=2
    minGustDuration=0.5
    minGustStrength=3
    minTimeBetweenGusts=0
    minUpInterpSpeed=2
    windTurbulenceScalarA=0.1
    windTurbulenceScalarB=1
    windAmplitude=<0 0 0>
    windFrequency=<0 0 0>
    windOffset=<0 0 0>
   }
   front_End {
    maxDownInterpSpeed=5
    maxGustDuration=2
    maxGustStrength=7
    maxTimeBetweenGusts=7
    maxUpInterpSpeed=5
    minDownInterpSpeed=2
    minGustDuration=0.5
    minGustStrength=3
    minTimeBetweenGusts=0
    minUpInterpSpeed=2
    windTurbulenceScalarA=0.1
    windTurbulenceScalarB=1
    windAmplitude=<0 0 0>
    windFrequency=<0 0 0>
    windOffset=<0 0 0>
   }
   onGround_Crouch_Moving {
    maxDownInterpSpeed=5
    maxGustDuration=2
    maxGustStrength=7
    maxTimeBetweenGusts=7
    maxUpInterpSpeed=5
    minDownInterpSpeed=2
    minGustDuration=0.5
    minGustStrength=3
    minTimeBetweenGusts=0
    minUpInterpSpeed=2
    windTurbulenceScalarA=0.1
    windTurbulenceScalarB=1
    windAmplitude=<0 0 0>
    windFrequency=<0 0 0>
    windOffset=<0 0 0>
   }
   onGround_Falling {
    maxDownInterpSpeed=5
    maxGustDuration=2
    maxGustStrength=7
    maxTimeBetweenGusts=7
    maxUpInterpSpeed=5
    minDownInterpSpeed=2
    minGustDuration=0.5
    minGustStrength=3
    minTimeBetweenGusts=0
    minUpInterpSpeed=2
    windTurbulenceScalarA=0.1
    windTurbulenceScalarB=1
    windAmplitude=<0 0 0>
    windFrequency=<0 0 0>
    windOffset=<0 0 0>
   }
   onGround_Moving {
    maxDownInterpSpeed=5
    maxGustDuration=2
    maxGustStrength=7
    maxTimeBetweenGusts=7
    maxUpInterpSpeed=5
    minDownInterpSpeed=2
    minGustDuration=0.5
    minGustStrength=3
    minTimeBetweenGusts=0
    minUpInterpSpeed=2
    windTurbulenceScalarA=0.1
    windTurbulenceScalarB=1
    windAmplitude=<0 0 0>
    windFrequency=<0 0 0>
    windOffset=<0 0 0>
   }
   onGround_Standing {
    maxDownInterpSpeed=5
    maxGustDuration=2
    maxGustStrength=7
    maxTimeBetweenGusts=7
    maxUpInterpSpeed=5
    minDownInterpSpeed=2
    minGustDuration=0.5
    minGustStrength=3
    minTimeBetweenGusts=0
    minUpInterpSpeed=2
    windTurbulenceScalarA=0.1
    windTurbulenceScalarB=1
    windAmplitude=<0 0 0>
    windFrequency=<0 0 0>
    windOffset=<0 0 0>
   }
   rail_Grinding {
    maxDownInterpSpeed=5
    maxGustDuration=8
    maxGustStrength=7
    maxTimeBetweenGusts=8
    maxUpInterpSpeed=5
    minDownInterpSpeed=2
    minGustDuration=0.2
    minGustStrength=3
    minTimeBetweenGusts=1
    minUpInterpSpeed=2
    windTurbulenceScalarA=0.1
    windTurbulenceScalarB=0.8
    windAmplitude=<20 65 42>
    windFrequency=<5 6 8.6>
    windOffset=<50 -500 -400>
   }
   riding_Moving {
    maxDownInterpSpeed=5
    maxGustDuration=2
    maxGustStrength=7
    maxTimeBetweenGusts=7
    maxUpInterpSpeed=5
    minDownInterpSpeed=2
    minGustDuration=0.5
    minGustStrength=3
    minTimeBetweenGusts=0
    minUpInterpSpeed=2
    windTurbulenceScalarA=0.1
    windTurbulenceScalarB=1
    windAmplitude=<0 0 0>
    windFrequency=<0 0 0>
    windOffset=<0 0 0>
   }
   riding_Moving_Targeting {
    maxDownInterpSpeed=5
    maxGustDuration=2
    maxGustStrength=7
    maxTimeBetweenGusts=7
    maxUpInterpSpeed=5
    minDownInterpSpeed=2
    minGustDuration=0.5
    minGustStrength=3
    minTimeBetweenGusts=0
    minUpInterpSpeed=2
    windTurbulenceScalarA=0.1
    windTurbulenceScalarB=1
    windAmplitude=<0 0 0>
    windFrequency=<0 0 0>
    windOffset=<0 0 0>
   }
   riding_Standing {
    maxDownInterpSpeed=5
    maxGustDuration=2
    maxGustStrength=7
    maxTimeBetweenGusts=7
    maxUpInterpSpeed=5
    minDownInterpSpeed=2
    minGustDuration=0.5
    minGustStrength=3
    minTimeBetweenGusts=0
    minUpInterpSpeed=2
    windTurbulenceScalarA=0.1
    windTurbulenceScalarB=1
    windAmplitude=<0 0 0>
    windFrequency=<0 0 0>
    windOffset=<0 0 0>
   }
   skydive_DivingDown {
    maxDownInterpSpeed=5
    maxGustDuration=2
    maxGustStrength=7
    maxTimeBetweenGusts=7
    maxUpInterpSpeed=5
    minDownInterpSpeed=2
    minGustDuration=0.5
    minGustStrength=3
    minTimeBetweenGusts=0
    minUpInterpSpeed=2
    windTurbulenceScalarA=0.1
    windTurbulenceScalarB=1
    windAmplitude=<0 0 0>
    windFrequency=<0 0 0>
    windOffset=<0 0 0>
   }
   skydive_FloatingDown {
    maxDownInterpSpeed=5
    maxGustDuration=2
    maxGustStrength=7
    maxTimeBetweenGusts=7
    maxUpInterpSpeed=5
    minDownInterpSpeed=2
    minGustDuration=0.5
    minGustStrength=3
    minTimeBetweenGusts=0
    minUpInterpSpeed=2
    windTurbulenceScalarA=0.1
    windTurbulenceScalarB=1
    windAmplitude=<0 0 0>
    windFrequency=<0 0 0>
    windOffset=<0 0 0>
   }
   skydive_GoingUp {
    maxDownInterpSpeed=5
    maxGustDuration=2
    maxGustStrength=7
    maxTimeBetweenGusts=7
    maxUpInterpSpeed=5
    minDownInterpSpeed=2
    minGustDuration=0.5
    minGustStrength=3
    minTimeBetweenGusts=0
    minUpInterpSpeed=2
    windTurbulenceScalarA=0.1
    windTurbulenceScalarB=1
    windAmplitude=<0 0 0>
    windFrequency=<0 0 0>
    windOffset=<0 0 0>
   }
   skydive_Parachuting {
    maxDownInterpSpeed=5
    maxGustDuration=2
    maxGustStrength=7
    maxTimeBetweenGusts=7
    maxUpInterpSpeed=5
    minDownInterpSpeed=2
    minGustDuration=0.5
    minGustStrength=3
    minTimeBetweenGusts=0
    minUpInterpSpeed=2
    windTurbulenceScalarA=0.1
    windTurbulenceScalarB=1
    windAmplitude=<0 0 0>
    windFrequency=<0 0 0>
    windOffset=<0 0 0>
   }
   skydive_Surfing {
    maxDownInterpSpeed=5
    maxGustDuration=2
    maxGustStrength=7
    maxTimeBetweenGusts=7
    maxUpInterpSpeed=5
    minDownInterpSpeed=2
    minGustDuration=0.5
    minGustStrength=3
    minTimeBetweenGusts=0
    minUpInterpSpeed=2
    windTurbulenceScalarA=0.1
    windTurbulenceScalarB=1
    windAmplitude=<0 0 0>
    windFrequency=<0 0 0>
    windOffset=<0 0 0>
   }
   swimming {
    maxDownInterpSpeed=5
    maxGustDuration=2
    maxGustStrength=7
    maxTimeBetweenGusts=7
    maxUpInterpSpeed=5
    minDownInterpSpeed=2
    minGustDuration=0.5
    minGustStrength=3
    minTimeBetweenGusts=0
    minUpInterpSpeed=2
    windTurbulenceScalarA=0.1
    windTurbulenceScalarB=1
    windAmplitude=<0 0 0>
    windFrequency=<0 0 0>
    windOffset=<0 0 0>
   }
   tactical_Sprint {
    maxDownInterpSpeed=5
    maxGustDuration=2
    maxGustStrength=7
    maxTimeBetweenGusts=7
    maxUpInterpSpeed=5
    minDownInterpSpeed=2
    minGustDuration=0.5
    minGustStrength=3
    minTimeBetweenGusts=0
    minUpInterpSpeed=2
    windTurbulenceScalarA=0.1
    windTurbulenceScalarB=1
    windAmplitude=<0 0 0>
    windFrequency=<0 0 0>
    windOffset=<0 0 0>
   }
   ziplining {
    maxDownInterpSpeed=5
    maxGustDuration=2
    maxGustStrength=7
    maxTimeBetweenGusts=7
    maxUpInterpSpeed=5
    minDownInterpSpeed=2
    minGustDuration=0.5
    minGustStrength=3
    minTimeBetweenGusts=0
    minUpInterpSpeed=2
    windTurbulenceScalarA=0.1
    windTurbulenceScalarB=1
    windAmplitude=<0 0 0>
    windFrequency=<0 0 0>
    windOffset=<0 0 0>
   }
  }
 }
]
